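; MVVM viewmodel for the Stat Manager UI, holding one stat's name, icon, color, value, level and display state.
; UpdateStaticDetails sets identity and max level, stores an opaque Slate-compatible color as an MVVM conversion workaround, then calls UpdateProgress.
; UpdateProgress derives fill percentage (full at max level) and the IsMaximum flag.
; Max value and level visibilities collapse when there is no positive maximum or only one level, respectively.
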(event Custom|UpdateProgress
  (params (Value :type Integer) (MaxValue :type Integer) (Level :type Integer))
  (Variables|Default|SetValue (Math|Conversions|ToInteger64(Integer) Value))
  (Variables|Default|SetMaxValue (Math|Conversions|ToInteger64(Integer) MaxValue))
  (Variables|Default|SetLevel (Math|Conversions|ToInteger64(Integer) Level))
  (Variables|Default|SetPercentage (Math|Float|SelectFloat 1.0 (Math|Float|SafeDivide (Math|Conversions|ToFloat(Integer64) (Variables|Default|GetValue)) (Math|Conversions|ToFloat(Integer64) (Variables|Default|GetMaxValue))) (== (Variables|Default|GetLevel) (Variables|Default|GetMaxLevel))))
  (Variables|Default|SetMaxValueVisibility (select (> (Variables|Default|GetMaxValue) 0) "HitTestInvisible" "Collapsed"))
  (Variables|Default|SetIsMaximum (CallFunction|IsMaximumStat)))

(event Custom|UpdateStaticDetails
  (params (Name :type Text) (Color :type Linear%20Color%20Structure) (MaxLevel :type Integer) (Icon%20Material :type Texture%202D%20Object%20Reference) (MaxValue :type Integer))
  (Variables|Default|SetName Name)
  (CallFunction|MakeColor Color)
  (Variables|Default|SetMaxLevel (Math|Conversions|ToInteger64(Integer) MaxLevel))
  (Variables|Default|SetIcon Icon%20Material)
  (Variables|Default|SetLevelVisibility (select (> MaxLevel 1) "HitTestInvisible" "Collapsed"))
  (CallFunction|UpdateProgress (Math|Conversions|ToInteger(Integer64) (Variables|Default|GetValue)) MaxValue (Math|Conversions|ToInteger(Integer64) (Variables|Default|GetLevel))))

(fn IsMaximumStat
  (if (> (Variables|Default|GetMaxValue) 0)
    (return (or (== (Variables|Default|GetLevel) (Variables|Default|GetMaxLevel)) (and (>= (Variables|Default|GetValue) (Variables|Default|GetMaxValue)) (<= (Variables|Default|GetMaxLevel) 1))))
    (else
      (return false))))

(fn MakeColor
  (params (Color_SpecifiedColor_R :type Float%20%28single-precision%29) (Color_SpecifiedColor_G :type Float%20%28single-precision%29) (Color_SpecifiedColor_B :type Float%20%28single-precision%29) (Color_SpecifiedColor_A :type Float%20%28single-precision%29))
  ;[0.2195 0.6584 1] We want to create a slate color (since MVVM can't turn a linear color into a Slate color, but you can easily drill down into the linear color from a slate color). Also, ignore alpha.
  (Variables|SetColor_R (Math|Color|MakeColor Color_SpecifiedColor_R Color_SpecifiedColor_G Color_SpecifiedColor_B)))
