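; Playspace component that replicates Stat Creator device details and disabled stats to clients and shows the player stat Dynamic UI scene during minigames.
; Server registrations update a GUID map, replicated as an array one tick later to batch changes.
; OnRep handlers rebuild client lookups and broadcast updates.
; Minigame state changes and playspace user events add or remove the scene, using timed retries when the minigame or scene isn't ready.
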(event EventBeginPlay
  ;[0.04337 1 0.127 0.187] Construction - Bind to Minigame Events, User Added Events
  (bind _success (CallFunction|BindMinigame true))
  (if _success
    (CallFunction|BindUserAdded true)
    (else
      (CallFunction|AttemptRebindMinigame))))

(event EventEndPlay
  (params (EndPlayReason :type EEndPlayReason%20Enum :default "Destroyed"))
  ;[1 0.05805 0 0.264] Deconstruction - Unbind from Minigame Events
  (CallFunction|BindMinigame false)
  (CallFunction|BindUserAdded false))

(event Custom|Attempt%20Rebind%20Minigame
  ;[0.04337 1 0.127 0.187] Construction - Bind to Minigame Events, User Added Events
  ;;[0.04337 1 0.127 0] Retry after delay if we fail to bind (likely attempt is too early)
  (Utilities|FlowControl|Delay 2.0)
  (bind _success (CallFunction|BindMinigame true))
  (if _success
    (CallFunction|BindUserAdded true)
    (else
      (CallFunction|AttemptRebindMinigame))))

(event Custom|Minigame%20State%20Changed
  (params (Minigame :type Fort%20Minigame%20Object%20Reference) (MinigameState :type EFortMinigameState%20Enum :default "PreGame"))
  (bind _returnvalue (WorldSubsystems|GetDynamicUIManager))
  (bind _stat_dynamic_ui_scene (Variables|Default|GetStat%20Dynamic%20UI%20Scene))
  ;[1 0.6099 0 0.2] Minigame State Changed - Determines whether we want to show or hide the UI.
  (switch Utilities|FlowControl|Switch|SwitchonEFortMinigameState MinigameState
    (:InProgress
      (Networking|CanExecuteCosmeticEvents
        (:True
          (DynamicUI|AddScenetoFirstLocalPlayer _returnvalue _stat_dynamic_ui_scene))))
    (:PostGameReset
      (Networking|CanExecuteCosmeticEvents
        (:True
          (DynamicUI|RemoveScenefromFirstLocalPlayer _returnvalue _stat_dynamic_ui_scene))))))

(event Custom|Replicate%20Details%20Next%20Frame
  ;[0 0.787 1 0.2] Server - Delayed Replication of Stat Details
  (Utilities|FlowControl|DelayUntilNextTick) ; Force a one frame delay\nto avoid lots of replication attempts during init
  (bind _values (Utilities|Map|Values (Variables|Default|GetServer%20Stat%20Creator%20Details)))
  (Variables|Default|SetReplicated%20Stat%20Details _values))

(fn AddSceneAfterDelay
  (bind _returnvalue (WorldSubsystems|GetDynamicUIManager))
  (bind _playercontrollertoadd (Variables|Default|GetPlayerControllerToAdd))
  (bind _stat_dynamic_ui_scene (Variables|Default|GetStat%20Dynamic%20UI%20Scene))
  (for _array_element _playercontrollertoadd
    (if (not (DynamicUI|IsSceneActive _returnvalue _stat_dynamic_ui_scene _array_element))
      (DynamicUI|AddScene (WorldSubsystems|GetDynamicUIManager) (Variables|Default|GetStat%20Dynamic%20UI%20Scene) _array_element)
      (if (not (DynamicUI|IsSceneActive _returnvalue _stat_dynamic_ui_scene _array_element))
        (Utilities|Time|SetTimerbyEvent (EventDispatchers|CreateEvent :!event AddSceneAfterDelay) 0.1)
        (else
          (Fortnite|FortShippingLog 0 "SCQX: PSD = After adding the scene, it now exists")))
      (else
        (Fortnite|FortShippingLog 0 "SCQX: PSD = There was already an active scene!"))))
  (Utilities|Array|Clear _playercontrollertoadd))

(fn BindMinigame
  (params (Bind :type Boolean))
  (bind _outputdelegate (EventDispatchers|CreateEvent :!event Minigame%20State%20Changed))
  (bind _cachedminigame (Variables|Default|GetCachedMinigame)
    (:then
      (if Bind
        (Minigame|BindEventtoStateChangedDelegate _cachedminigame _outputdelegate)
        (CallFunction|MinigameStateChanged (Variables|Default|GetCachedMinigame) (Variables|/Script/FortniteGame.FortMinigame|GetCurrentState (Variables|Default|GetCachedMinigame))) ; Call Minigame State Changed to cover starting state..
        (else
          (Minigame|UnbindEventfromStateChangedDelegate _cachedminigame _outputdelegate)))
      (return true))
    (:else
      (if (Variables|GetBind)
        (return false)
        (else
          (return true))))))

(fn BindUserAdded
  (params (Bind :type Boolean))
  (bind _outputdelegate (EventDispatchers|CreateEvent :!event PlayspaceUserAdded))
  (bind _outputdelegate_2 (EventDispatchers|CreateEvent :!event PlayspaceUserRemoved))
  (Networking|CanExecuteCosmeticEvents
    (:True
      (bind _asplayspace (Utilities|Casting|CastToPlayspace (Components|GetOwner)))
      (Fortnite|FortShippingLog 0 "SCQX: PSD - attempting to bind user events")
      (if Bind
        (Playspace|BindEventtoNotifyPlayspaceUserAdded _asplayspace _outputdelegate)
        (Playspace|BindEventtoNotifyPlayspaceUserRemoved _asplayspace _outputdelegate_2)
        (else
          (Playspace|UnbindEventfromNotifyPlayspaceUserAdded _asplayspace _outputdelegate)
          (Playspace|UnbindEventfromNotifyPlayspaceUserRemoved _asplayspace _outputdelegate_2))))))

(fn DeregisterDevice
  (params (Stat%20Creator%20GUID :type Guid%20Structure%20%28by%20ref%29))
  (bind _returnvalue (Utilities|Map|Remove (Variables|Default|GetServer%20Stat%20Creator%20Details) Stat%20Creator%20GUID))
  (if _returnvalue
    (CallFunction|ReplicateDetailsNextFrame)))

(fn GetStatDetails
  (return (Variables|Default|GetStat%20Creator%20Details) (Variables|Default|GetDisabled%20Stats)))

(fn OnRep_CachedMinigame
  (Variables|Default|GetCachedMinigame
    (:then
      (CallFunction|BindMinigame true))
    (:else)))

(fn OnRep_ReplicatedStatDetails
  (Utilities|Map|Clear (Variables|Default|GetStat%20Creator%20Details))
  (for _array_element (Variables|Default|GetReplicated%20Stat%20Details)
    (Utilities|Map|Add (Variables|Default|GetStat%20Creator%20Details) (Utilities|Struct|BreakStatCreatorDetails _array_element) _array_element))
  (bind _statdetails (CallFunction|GetStatDetails))
  (Default|CallOnClientDetailsUpdated _statdetails))

(fn OnRep_ReplicatedDisabledFilters
  (bind _localdisabledset (LocalVariables|GetLocalDisabledSet))
  (Utilities|Set|AddItems _localdisabledset (Variables|Default|GetReplicatedDisabledFilters))
  (bind _output_get (Variables|Default|SetDisabled%20Stats _localdisabledset))
  (Default|CallOnClientEnabledStateChanged _output_get))

(fn PlayspaceUserAdded
  (params (AddedUser :type Playspace%20User%20Structure%20%28by%20ref%29))
  (bind (_userid _leafplayspace _playerstatecached) (Utilities|Struct|BreakPlayspaceUser AddedUser))
  (Fortnite|FortShippingLog 0 "SCQX: PSD - got into playspace user added  ")
  (bind _asplayer_controller (Utilities|Casting|CastToPlayerController (Fortnite|GetControllerfromActor _playerstatecached))
    (:then
      (DynamicUI|RemoveScene (WorldSubsystems|GetDynamicUIManager) (Variables|Default|GetStat%20Dynamic%20UI%20Scene) _asplayer_controller)
      (bind _returnvalue (Minigame|GetMinigameforActor _playerstatecached))
      (Utilities|IsValid _returnvalue
        (:Is%20Valid
          (Minigame|IsInProgress _returnvalue
            (:Yes
              (Utilities|Array|Add (Variables|Default|GetPlayerControllerToAdd) _asplayer_controller)
              (Utilities|Time|SetTimerbyEvent (EventDispatchers|CreateEvent :!event AddSceneAfterDelay) 0.1))
            (:No
              _))
          (return))
        (:Is%20Not%20Valid
          (Fortnite|FortShippingLog 0 "SCQX PSD - minigame not found before progress check")
          (Utilities|Array|Add (Variables|Default|GetPlayerSpaceUsersToAdd) AddedUser)
          (Utilities|Time|SetTimerbyEvent (EventDispatchers|CreateEvent :!event RetryPlayspaceUserAdded) 2.0))))
    (:CastFailed
      (Fortnite|FortShippingLog 0 "SCQX: PSD - failing to cast to Player controller "))))

(fn PlayspaceUserRemoved
  (params (RemovedUser :type Playspace%20User%20Structure%20%28by%20ref%29))
  (bind (_userid _leafplayspace _playerstatecached) (Utilities|Struct|BreakPlayspaceUser RemovedUser))
  (bind _asplayer_controller (Utilities|Casting|CastToPlayerController (Fortnite|GetControllerfromActor _playerstatecached)))
  (DynamicUI|RemoveScene (WorldSubsystems|GetDynamicUIManager) (Variables|Default|GetStat%20Dynamic%20UI%20Scene) _asplayer_controller)
  (return))

(fn RegisterOrUpdateDevice
  (params (Stat%20Creator%20GUID :type Guid%20Structure) (Value :type Stat%20Creator%20Details%20Structure%20%28by%20ref%29))
  (Utilities|Map|Add (Variables|Default|GetServer%20Stat%20Creator%20Details) Stat%20Creator%20GUID Value)
  (CallFunction|ReplicateDetailsNextFrame))

(fn RetryPlayspaceUserAdded
  (bind _playerspaceuserstoadd (Variables|Default|GetPlayerSpaceUsersToAdd))
  (Fortnite|FortShippingLog 0 "SCQX:  PSD - Reached \"retry playspace user added\"")
  (bind _output_get (LocalVariables|SetUserToAdd (Utilities|Array|Get(acopy) _playerspaceuserstoadd)))
  (Utilities|Array|RemoveIndex _playerspaceuserstoadd)
  (CallFunction|PlayspaceUserAdded _output_get))

(fn SetEnabled
  (params (Value%20Stat%20Filter :type Fort%20Minigame%20Stat%20Filter%20Object%20Reference) (Enabled%20State :type Boolean) (ItemToFind :type Gameplay%20Tag%20Structure%20%28by%20ref%29))
  (bind _replicateddisabledfilters (Variables|Default|GetReplicatedDisabledFilters))
  (if (not Enabled%20State)
    (if (not (Utilities|Array|ContainsItem (Variables|Default|GetReplicatedDisabledFilters) ItemToFind))
      (Utilities|Array|Add _replicateddisabledfilters ItemToFind))
    (else
      (bind _returnvalue (Utilities|Array|RemoveItem (Variables|Default|GetReplicatedDisabledFilters) ItemToFind))
      (if _returnvalue
        _)))
  (Variables|Default|SetReplicatedDisabledFilters _replicateddisabledfilters))

(fn SetMinigame
  (params (NewMinigame :type Fort%20Minigame%20Object%20Reference))
  (Utilities|SwitchHasAuthority
    (:Authority
      (Variables|Default|SetCachedMinigame NewMinigame))))
