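# StatAnimationContainer is a HUD widget that shows one creator-defined Stat Manager stat, with optional intro and outro animations.
# It async-loads the stat's configured widget into its canvas and feeds it an MVVM_UEFN_Stat view model.
# Minigame stat-change events re-read value and level by player, team or group scope, then show or close it per display rules.
# Minigame binding retries on a delay, and visibility changes broadcast OnVisualStateUpdated with the widget's width.
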WidgetBlueprint /CRD_StatManager/UI/StatAnimationContainer
animations [
 Intro {
  binding_count=1
  length_seconds=0.5
  object=&self:Intro
  sequence {
   bindings [Panel {
     id=#1
     possessed_class=CanvasPanel
     tracks [
      {
       type=MovieSceneFloatTrack
       display_name=Render Opacity
       sections [{type=MovieSceneFloatSection, end_frame=5, start_frame=0, channels [None {type=MovieSceneScriptingFloatChannel, keys [[0,0],[5,1]]}]}]
      }
      {
       type=MovieScene2DTransformTrack
       sections [{
         type=MovieScene2DTransformSection
         end_frame=10
         start_frame=0
         channels [
          Translation.X {type=MovieSceneScriptingFloatChannel, keys [[0,-50],[10,0]]}
          Translation.Y {type=MovieSceneScriptingFloatChannel}
          Angle {type=MovieSceneScriptingFloatChannel}
          Scale.X {type=MovieSceneScriptingFloatChannel}
          Scale.Y {type=MovieSceneScriptingFloatChannel}
          Shear.X {type=MovieSceneScriptingFloatChannel}
          Shear.Y {type=MovieSceneScriptingFloatChannel}
         ]
        }]
      }
     ]
    }]
   tracks [{type=MovieSceneEventTrack, sections [{type=MovieSceneEventTriggerSection, channels [None {type=MovieSceneScriptingEventChannel}]}]}]
  }
 }
 Outro {
  binding_count=1
  length_seconds=0.25
  object=&self:Outro
  sequence {bindings [Panel {
     id=#2
     possessed_class=CanvasPanel
     tracks [{
       type=MovieSceneFloatTrack
       display_name=Render Opacity
       sections [{type=MovieSceneFloatSection, end_frame=5, start_frame=0, channels [None {type=MovieSceneScriptingFloatChannel, keys [[0,1],[5,0]]}]}]
      }]
    }]}
 }
]
graphs [
 Bind Minigame {nodes [
   Bind Minigame() {is=title K2Node_FunctionEntry}
   n1 = GetMinigame!(ContextActor:Actor Object Reference=GetOwningPlayer()) -> (ReturnValue:Fort Minigame Object Reference)
   > n2 = Is Valid(InputObject=@n1)
   FortShippingLog!(
    execute=@n2.Is Not Valid
    WorldContextObject:Object Reference
    InString:String=SCQX: SAC - attempting to rebind minigame
    bLogAsWarning:bool=false
   ) -> (then)
   > Attempt Rebind Minigame()
   n3 = Set StatComponent(
    execute=@n2.Is Valid
    StatComponent:Fort Minigame Stats Component Object Reference=GetMinigameStatsComponent!(ContextActor:Actor Object Reference=@n1) -> (ReturnValue:Fort Minigame Stats Component Object Reference)
   ) -> (Output_Get:Fort Minigame Stats Component Object Reference)
   > Bind Event to Stat Changed Delegate(self:Fort Minigame Stats Component Object Reference=@n3, Delegate=Create Event()) {is=title K2Node_AddDelegate}
  ]}
 Create and Set View Model {nodes [
   Create and Set View Model() {is=title K2Node_FunctionEntry}
   > Get!(execute:Exec) -> (then:Exec, else:Exec, LoadedWidget:User Widget Object Reference)
   > n1 = Get!(execute) -> (then, else, ViewModel:Device - Stat View Model Object Reference)
   n2 = Get View From User Widget!(
    self:Viewmodel Engine Subsytem Object Reference=Get MVVMSubsystem() -> (ReturnValue:Viewmodel Engine Subsytem Object Reference) {is=title K2Node_GetEngineSubsystem}
    UserWidget:User Widget Object Reference=Get LoadedWidget() -> (LoadedWidget:User Widget Object Reference)
   ) -> (ReturnValue:MVVM View Object Reference) {is=title K2Node_CallFunction}
   Is Valid(exec=[@n1.then, @n4.then], InputObject=@n2)
   > SetViewModel!(
    execute
    self:MVVM View Object Reference=@n2
    ViewModelName:Name=MVVM_UEFN_Stat
    ViewModel:Notify Field Value Changed Interface=Get ViewModel() -> (ViewModel:Device - Stat View Model Object Reference)
   ) -> (then, ReturnValue:bool)
   n3 = Construct Object from Class(execute=@n1.else, Class=&/MVVM_UEFN_Stat.MVVM_UEFN_Stat_C) -> (ReturnValue:Device - Stat View Model Object Reference)
   > n4 = Set ViewModel(ViewModel:Device - Stat View Model Object Reference=@n3) -> (Output_Get:Device - Stat View Model Object Reference)
  ]}
 EventGraph {nodes [
   Attempt Rebind Minigame() {is=title K2Node_CustomEvent}
   > Delay(Duration=2)
   > Bind Minigame() {is=title K2Node_CallFunction}
   Display Widget() {is=title K2Node_CustomEvent}
   > n1 = Branch(Condition=Get IsActive() -> (IsActive:bool))
   > n2 = Is Closing or Closed!(" ":Exec) -> (Yes:Exec, No:Exec) {is=title K2Node_MacroInstance}
   > Stop Outro()
   > Update Visibility!(execute, self, On:bool=true) -> (then)
   > n3 = Play Intro()
   Branch(
    execute=[@n3.then, @n2.No, @n1.else]
    Condition="Equal (Enum)"(
     A:E_StatManager_ShowOnHUD Enum=Get Stat Bar Shown on HUD() -> (Stat Bar Shown on HUD:E_StatManager_ShowOnHUD Enum=NewEnumerator0)
     B:E_StatManager_ShowOnHUD Enum=NewEnumerator3
    )
   )
   > RetriggerableDelay(Duration=Break Stat Creator Details!(StatCreatorDetails:"Stat Creator Details Structure (by ref)"=Get StatDetails() -> (StatDetails:Stat Creator Details Structure)) -> (
     ValueFilter_53_#3:Fort Minigame Stat Filter Object Reference
     LevelFilter_54_#4:Fort Minigame Stat Filter Object Reference
     Scope_50_#5:E_StatManager_Scope Enum
     ShowHUDInList_52_#6:bool=~
     DisplayTime_21_#7:double=~
     IntroAnimation_58_#8:EStatCreatorIntroAnimation Enum
     OutroAnimation_61_#9:EStatCreatorOutroAnimation Enum
    ) {is=title K2Node_BreakStruct}.DisplayTime_21_#7)
   > n4 = Is Closing or Closed()
   n5 = Play Outro!(execute=[@n4.No, @n15.then], self) -> (then, OutroIsPlaying:bool)
   > n6 = Branch(Condition=@n5)
   > n7 = RetriggerableDelay(Duration=GetEndTime(self=Get Outro() -> (Outro:Widget Animation Object Reference)))
   Is Closing or Closed(" "=[@n7.then, @n8.then])
   > Set OutroIsPlaying(OutroIsPlaying:bool=false) -> (Output_Get:bool)
   > Update Visibility(On=false)
   n8 = Set OutroIsPlaying(execute=@n6.else, OutroIsPlaying=true)
   n9 = Update Visibility!() -> (OutputDelegate, then, On:bool=~) {is=title K2Node_CustomEvent}
   n10 = "float > float"(A=Break Stat Creator Details(StatCreatorDetails=Get StatDetails()) -> (ShowHUDInList_52_#6=~, DisplayTime_21_#7=~).DisplayTime_21_#7)
   > SetVisibility(InVisibility=Select(Option 0=Select(Option 0=Collapsed, Option 1=Hidden, Index=@n10), Option 1=HitTestInvisible, Index=@n9.On))
   > DelayUntilNextTick!(execute) -> (then)
   n11 = Self-Reference()
   > Call On Visual State Updated!(
    execute:Exec
    self:Stat Animation Container Object Reference
    Widget:Stat Animation Container Object Reference=Self-Reference()
    Width:double=Select(
     Option 0:double=0
     Option 1:double=GetDesiredSize!(self:Widget Object Reference=@n11) -> (ReturnValue_X:double, ReturnValue_Y:double)
     Index=@n9.On
    ) -> (ReturnValue:double)
    TrackedFilter:Fort Minigame Player Defined Stat Filter Object Reference=Get Owning Stat Filter() -> (Owning Stat Filter:Fort Minigame Player Defined Stat Filter Object Reference)
    IsActive:bool=Get IsActive()
   ) -> (then:Exec) {is=title K2Node_CallDelegate}
   Update View Model Static Details() {is=title K2Node_CustomEvent}
   > n12 = Load Device Icon!(" ":Exec) -> (Complete:Exec, Icon:Texture 2D Object Reference) {is=title K2Node_MacroInstance}
   n13 = Break Stat Creator Details!(StatCreatorDetails=Get StatDetails()) -> (
    Name_41_#10:Text
    StatColor_7_#11:Linear Color Structure
    MaxLevel_12_#12:int
   )
   > UpdateStaticDetails!(
    execute
    self:Device - Stat View Model Object Reference=Get ViewModel()
    Name:Text=@n13
    Color:Linear Color Structure=@n13.StatColor_7_#11
    MaxLevel:int=@n13.MaxLevel_12_#12
    Icon Material:Texture 2D Object Reference=@n12
    MaxValue:int=Get Max Value!(Level:"Integer (by ref)"=Get Level() -> (Level:int=0)) -> (MaxValue:int=0) {is=title K2Node_CallFunction}
   ) -> (then)
   n14 = Force Close!() -> (OutputDelegate, then, StillActive:bool=~) {is=title K2Node_CustomEvent}
   > n15 = Set IsActive(IsActive:bool=@n14.StillActive) -> (Output_Get:bool)
   n16 = Initialize From StatDetails!() -> (OutputDelegate, then, Details:Stat Creator Details Structure) {is=title K2Node_CustomEvent}
   > Set StatDetails(StatDetails:Stat Creator Details Structure=@n16.Details) -> (Output_Get:Stat Creator Details Structure)
   > Set IsActive(IsActive=true)
   n17 = Break Stat Creator Details!(StatCreatorDetails=@n16.Details) -> (
    ValueFilter_53_#3:Fort Minigame Stat Filter Object Reference
    LevelFilter_54_#4:Fort Minigame Stat Filter Object Reference
    Scope_50_#5:E_StatManager_Scope Enum
    HUDPriority_19_#13:int
    ShowHUDInList_52_#6:bool=~
    DisplayTime_21_#7:double=~
    IntroAnimation_58_#8:EStatCreatorIntroAnimation Enum
    OutroAnimation_61_#9:EStatCreatorOutroAnimation Enum
   )
   > n18 = Cast To FortMinigamePlayerDefinedStatFilter!(execute, Object=@n17) -> (then, CastFailed, AsFort Minigame Player Defined Stat Filter:Fort Minigame Player Defined Stat Filter Object Reference) {is=title K2Node_DynamicCast}
   > n19 = Set Owning Stat Filter(Owning Stat Filter:Fort Minigame Player Defined Stat Filter Object Reference=@n18) -> (Output_Get:Fort Minigame Player Defined Stat Filter Object Reference)
   n20 = Cast To FortMinigamePlayerDefinedStatFilter(execute=[@n19.then, @n18.CastFailed], Object=@n17.LevelFilter_54_#4)
   > n21 = Set Owning Stat Level Filter(Owning Stat Level Filter:Fort Minigame Player Defined Stat Filter Object Reference=@n20) -> (Output_Get:Fort Minigame Player Defined Stat Filter Object Reference)
   Set Stat Bar Shown on HUD(
    execute=[@n21.then, @n20.CastFailed]
    Stat Bar Shown on HUD:E_StatManager_ShowOnHUD Enum=GetStatBarShownFromDisplayTime!(InputPin:double=@n17.DisplayTime_21_#7) -> (ReturnValue:E_StatManager_ShowOnHUD Enum=NewEnumerator0)
   ) -> (Output_Get:E_StatManager_ShowOnHUD Enum=NewEnumerator0)
   > n22 = Widget Has Changed!(execute:Exec) -> (True:Exec, false:Exec) {is=title K2Node_MacroInstance}
   > SetVisibility(InVisibility=Collapsed)
   > n23 = CreateWidgetAsync!(
    execute:Exec
    UserWidgetSoftClass:User Widget Soft Class Reference=Break Stat Creator Details!(StatCreatorDetails=Get StatDetails()) -> (UserWidget_38_#14:User Widget Soft Class Reference)
    OwningPlayer:Player Controller Object Reference
    bSuspendInputUntilComplete:bool=false
   ) -> (then:Exec, OnComplete:Exec, OnFailure:Exec, UserWidget:User Widget Object Reference) {is=title K2Node_AsyncAction}
   n24 = Set LoadedWidget(execute=@n23.OnComplete, LoadedWidget:User Widget Object Reference=@n23) -> (Output_Get:User Widget Object Reference)
   > n25 = AddChildToCanvas!(execute, self:Canvas Panel Object Reference=Get Panel() -> (Panel:Canvas Panel Object Reference), Content:Widget Object Reference=@n24) -> (then, ReturnValue:Canvas Panel Slot Object Reference)
   > Set bAutoSize!(execute, bAutoSize:bool=true, self:Canvas Panel Slot Object Reference=@n25) -> (then, Output_Get:bool)
   > n26 = Branch(Condition=NOT Boolean(A=@n17.ShowHUDInList_52_#6))
   > SetAnchors!(
    execute
    self:Canvas Panel Slot Object Reference=@n25
    InAnchors:Anchors Structure=Make Anchors!(Minimum:Vector 2D Structure=(X=0,Y=0), Maximum:Vector 2D Structure=(X=1,Y=1)) -> (Anchors:Anchors Structure) {is=title K2Node_MakeStruct}
   ) -> (then)
   > n27 = SetOffsets!(
    execute
    self:Canvas Panel Slot Object Reference=@n25
    InOffset_Left:float=0
    InOffset_Top:float=0
    InOffset_Right:float=0
    InOffset_Bottom:float=0
   ) -> (then)
   n28 = Create and Set View Model(execute=[@n27.then, @n26.else]) {is=title K2Node_CallFunction}
   Update View Model Static Details(execute=[@n28.then, @n22.false]) {is=title K2Node_CallFunction}
   > Update ProgressFromStats!(execute, self, From Initialization:bool=true) -> (then)
   OnAnyStatChanged() {is=title K2Node_CustomEvent}
   > Branch(Condition=Get IsActive())
   > Update ProgressFromStats(From Initialization=false)
   Event Construct()
   > Bind Minigame()
   Event Destruct()
   n29 = GetFirstLocalFortPlayerController!(bIncludeNonSpectators:bool=true, bIncludeSpectators:bool=false) -> (ReturnValue:Fort Player Controller Object Reference)
   > Is Valid(InputObject=@n29)
   n30 = GetMinigame(ContextActor=@n29)
   > Is Valid(InputObject=@n30)
   > Unbind Event from Stat Changed Delegate(self:Fort Minigame Stats Component Object Reference=GetMinigameStatsComponent(ContextActor=@n30), Delegate=Create Event()) {is=title K2Node_RemoveDelegate}
  ]}
 Get Max Value {nodes [
   n1 = Get Max Value!() -> (then, Level:"Integer (by ref)") {is=title K2Node_FunctionEntry}
   n2 = Break Stat Creator Details!(StatCreatorDetails=Get StatDetails()) -> (MaxValues_37_#15:Array of Integers, MaxLevel_12_#12:int)
   > n3 = Branch(Condition=Is Not Empty!(TargetArray:Array of Integers=@n2) -> (ReturnValue:bool))
   n4 = Break Stat Creator Details(StatCreatorDetails=Get StatDetails())
   > n5 = Branch(Condition="integer < integer"(A=@n4.MaxLevel_12_#12, B=2))
   > Return Node!(execute, MaxValue:int="Get (a ref)"(Array=@n4, Dimension 1=Last Index!(TargetArray:Array of Integers=@n4) -> (ReturnValue=0)))
   n6 = "Min (Integer)"(A=Last Index(TargetArray=@n2) -> (ReturnValue=0), B=@n2.MaxLevel_12_#12) -> (ReturnValue=0)
   n7 = int + int(A=@n1)
   Return Node(execute=@n5.else, MaxValue="Get (a ref)"(Array=@n2, Dimension 1="Clamp (Integer)"(Value=@n7, Min=0, Max=@n6) -> (ReturnValue=0)))
   Return Node(execute=@n3.else, MaxValue=-1)
  ]}
 GetPriority {nodes [
   GetPriority() {is=title K2Node_FunctionEntry}
   > Return Node!(
    execute
    Priority:int=Break Stat Creator Details!(StatCreatorDetails=Get StatDetails()) -> (HUDPriority_19_#13:int)
   )
  ]}
 GetStat {nodes [
   n1 = GetStat!() -> (then, InFilter:Fort Minigame Stat Filter Object Reference) {is=title K2Node_FunctionEntry}
   > Is Valid(InputObject=@n1)
   > Get!(execute) -> (then, else, StatComponent:Fort Minigame Stats Component Object Reference)
   > n2 = Cast To FortPlayerState!(
    execute
    Object=Get PlayerState!(self:Controller Object Reference=GetFirstLocalFortPlayerController(bIncludeNonSpectators=true, bIncludeSpectators=false)) -> (PlayerState:Player State Object Reference)
   ) -> (then, CastFailed, AsFort Player State:Fort Player State Object Reference) {is=title K2Node_DynamicCast}
   > Set LocalPlayerState(LocalPlayerState:Fort Player State Object Reference=@n2) -> (Output_Get:Fort Player State Object Reference)
   > n3 = Switch on E_StatManager_Scope!(
    execute
    Selection:E_StatManager_Scope Enum=Break Stat Creator Details!(StatCreatorDetails=Get StatDetails()) -> (Scope_50_#5:E_StatManager_Scope Enum)
   ) -> (NewEnumerator0:Exec, NewEnumerator1:Exec, NewEnumerator2:Exec) {is=title K2Node_SwitchEnum}
   n4 = Get LocalPlayerState() -> (LocalPlayerState:Fort Player State Object Reference)
   > Return Node!(
    execute
    ReturnValue:int=GetStatForPlayerId!(
     self:Fort Minigame Stats Component Object Reference=Get StatComponent() -> (StatComponent:Fort Minigame Stats Component Object Reference)
     PlayerId:"Unique Net Id Structure (by ref)"=Get Unique Net Id!(self:Player State Object Reference=@n4) -> (ReturnValue:Unique Net Id Structure) {is=title K2Node_CallFunction}
     InFilter:Fort Minigame Stat Filter Object Reference=Get InFilter() -> (InFilter:Fort Minigame Stat Filter Object Reference)
     bGameTotal:bool=false
    ) -> (ReturnValue:int=0)
   )
   n5 = GetActorTeam!(Actor:Actor Object Reference=Get LocalPlayerState()) -> (ReturnValue:Byte=0)
   Return Node(
    execute=@n3.NewEnumerator1
    ReturnValue=GetStatForPlayerBucket!(
     self:Fort Minigame Stats Component Object Reference=Get StatComponent()
     BucketIndex:int=TeamToIndex!(Team:Byte=@n5) -> (Index:int=0)
     InFilter:Fort Minigame Stat Filter Object Reference=Get InFilter()
     bGameTotal:bool=false
     bIncludeInactivePlayers:bool=false
    ) -> (ReturnValue:int=0)
   )
   Return Node(
    execute=@n3.NewEnumerator2
    ReturnValue=GetStatForGroup!(
     self:Fort Minigame Stats Component Object Reference=Get StatComponent()
     InFilter:Fort Minigame Stat Filter Object Reference=Get InFilter()
     bGameTotal:bool=false
    ) -> (ReturnValue:int=0)
   )
  ]}
 GetStatBarShownFromDisplayTime {nodes [
   n1 = GetStatBarShownFromDisplayTime!() -> (then, InputPin:double=~) {is=title K2Node_FunctionEntry}
   n2 = "float > float"(A=@n1, B=-1)
   > Return Node(ReturnValue:E_StatManager_ShowOnHUD Enum=Select(
     Option 0:E_StatManager_ShowOnHUD Enum=Select(Option 0=NewEnumerator0, Option 1=NewEnumerator1, Index=@n2)
     Option 1:E_StatManager_ShowOnHUD Enum=NewEnumerator3
     Index="float > float"(A=@n1)
    ) -> (ReturnValue:E_StatManager_ShowOnHUD Enum))
  ]}
 IsClosingOrClosed {nodes [
   Inputs!() -> (" ":Exec)
   n1 = Get Visibility() -> (Visibility:ESlateVisibility Enum=Visible)
   > n2 = Branch(Condition=OR Boolean!(
     A=Get OutroIsPlaying() -> (OutroIsPlaying:bool)
     B="Equal (Enum)"(A:ESlateVisibility Enum=@n1, B:ESlateVisibility Enum=Collapsed)
     C:bool="Equal (Enum)"(A=@n1, B=Hidden)
    ) -> (ReturnValue))
   > Outputs!(Yes:Exec, No:Exec=@n2.else)
  ]}
 LoadDeviceIcon {nodes [
   Inputs()
   n1 = Break Stat Creator Details!(StatCreatorDetails=Get StatDetails()) -> (
    ValueFilter_53_#3:Fort Minigame Stat Filter Object Reference
    LevelFilter_54_#4:Fort Minigame Stat Filter Object Reference
    Name_41_#10:Text
    Icon_45_#16:Texture 2D Soft Object Reference
    Scope_50_#5
    StatColor_7_#11:Linear Color Structure
    MaxLevel_12_#12:int
    ShowHUDInList_52_#6:bool=~
    IntroAnimation_58_#8:EStatCreatorIntroAnimation Enum
    OutroAnimation_61_#9:EStatCreatorOutroAnimation Enum
   )
   > n2 = Branch(Condition=IsValidSoftObjectReference(SoftObjectReference=@n1.Icon_45_#16))
   > n3 = Async Load Asset(Asset=@n1.Icon_45_#16)
   n4 = Cast To Texture2D(execute=@n3.Completed, Object=@n3)
   Outputs!(Complete:Exec=[@n4.then, @n4.CastFailed, @n2.else], Icon:Texture 2D Object Reference=@n4)
  ]}
 OnVisualStateUpdated {nodes [OnVisualStateUpdated!() -> (
    then
    Widget:Stat Animation Container Object Reference
    Width:double=~
    TrackedFilter:Fort Minigame Player Defined Stat Filter Object Reference
    IsActive:bool=~
   ) {is=title K2Node_FunctionEntry}]}
 Play Intro {nodes [
   Play Intro() {is=title K2Node_FunctionEntry}
   > n1 = Switch on EStatCreatorIntroAnimation!(
    execute
    Selection:EStatCreatorIntroAnimation Enum=Break Stat Creator Details!(StatCreatorDetails=Get StatDetails()) -> (IntroAnimation_58_#8)
   ) -> (NewEnumerator0:Exec, NewEnumerator1:Exec) {is=title K2Node_SwitchEnum}
   PlayAnimationForward(execute=@n1.NewEnumerator1, InAnimation=Get Intro())
  ]}
 Play Outro {nodes [
   Play Outro() {is=title K2Node_FunctionEntry}
   > n1 = Switch on EStatCreatorOutroAnimation!(
    execute
    Selection:EStatCreatorOutroAnimation Enum=Break Stat Creator Details!(StatCreatorDetails=Get StatDetails()) -> (OutroAnimation_61_#9:EStatCreatorOutroAnimation Enum)
   ) -> (NewEnumerator0:Exec, NewEnumerator1:Exec) {is=title K2Node_SwitchEnum}
   > n2 = Set OutroIsPlaying(OutroIsPlaying=false)
   > Return Node!(execute, OutroIsPlaying:bool=@n2)
   PlayAnimationForward(execute=@n1.NewEnumerator1, InAnimation=Get Outro())
   > n3 = Set OutroIsPlaying(OutroIsPlaying=true)
   > Return Node(OutroIsPlaying=@n3)
  ]}
 Stop Outro {nodes [
   Stop Outro() {is=title K2Node_FunctionEntry}
   > StopAnimation(InAnimation=Get Outro())
   > Set OutroIsPlaying(OutroIsPlaying=false)
  ]}
 Update ProgressFromStats {nodes [
   Update ProgressFromStats!() -> (then, From Initialization:bool=~) {is=title K2Node_FunctionEntry}
   > Set NewValue(NewValue:int=GetStat!(InFilter:Fort Minigame Stat Filter Object Reference=Get Owning Stat Filter()) -> (ReturnValue:int=0) {is=title K2Node_CallFunction}) -> (Output_Get:int=0)
   > Set NewLevel(NewLevel:int=GetStat(InFilter=Get Owning Stat Level Filter() -> (Owning Stat Level Filter:Fort Minigame Player Defined Stat Filter Object Reference)) -> (ReturnValue=0)) -> (Output_Get:int=0)
   n1 = Get NewLevel() -> (NewLevel:int=0)
   > UpdateProgress!(
    execute
    self:Device - Stat View Model Object Reference=Get ViewModel()
    Value:int=Get NewValue() -> (NewValue:int=0)
    MaxValue:int=Get Max Value(Level=@n1) -> (MaxValue=0) {is=title K2Node_CallFunction}
    Level:int=@n1
   ) -> (then)
   n2 = OR Boolean!(
    A="Not Equal (Integer)"(A=Get Value() -> (Value:int=0), B=Get NewValue() -> (NewValue=0))
    B="Not Equal (Integer)"(A=Get Level() -> (Level=0), B=Get NewLevel() -> (NewLevel=0))
   ) -> (ReturnValue)
   n3 = Get From Initialization() -> (From Initialization:bool)
   n4 = OR Boolean(A="integer > integer"(A=Get NewValue() -> (NewValue=0)), B="integer > integer"(A=Get NewLevel() -> (NewLevel=0), B=1))
   n5 = Select!(
    NewEnumerator0:bool=true
    NewEnumerator1:bool=@n4
    NewEnumerator2:bool=false
    NewEnumerator3:bool=@n4
    Index:E_StatManager_ShowOnHUD Enum=Get Stat Bar Shown on HUD() -> (Stat Bar Shown on HUD=NewEnumerator0)
   ) -> (ReturnValue:bool)
   > n6 = Branch(Condition=AND Boolean!(A=OR Boolean(A=@n2, B=@n3), B=Get IsActive(), C:bool=@n5) -> (ReturnValue))
   n7 = Get NewValue() -> (NewValue=0)
   > Set Value(Value:int=@n7) -> (Output_Get:int=0)
   n8 = Get NewLevel() -> (NewLevel=0)
   > Set Level(Level:int=@n8) -> (Output_Get:int=0)
   > Display Widget() {is=title K2Node_CallFunction}
   Set Value(execute=@n6.else, Value=@n7) -> (Output_Get=0)
   > Set Level(Level=@n8) -> (Output_Get=0)
   > Branch(Condition=NOT Boolean(A=@n5))
   > Force Close!(execute, self, StillActive:bool=true) -> (then) {is=title K2Node_CallFunction}
  ]}
 WidgetHasChanged {nodes [
   Inputs!() -> (execute:Exec)
   > n1 = Get!(execute) -> (then, else, LoadedWidget:User Widget Object Reference)
   > n2 = Get!(execute) -> (then, else, Owning Stat Filter:Fort Minigame Player Defined Stat Filter Object Reference)
   n3 = Break Stat Creator Details!(StatCreatorDetails=Get StatDetails()) -> (
    ValueFilter_53_#3:Fort Minigame Stat Filter Object Reference
    LevelFilter_54_#4:Fort Minigame Stat Filter Object Reference
    Name_41_#10:Text
    Icon_45_#16:Texture 2D Soft Object Reference
    Scope_50_#5:E_StatManager_Scope Enum
    UserWidget_38_#14:User Widget Soft Class Reference
    ShowHUDInList_52_#6:bool=~
    IntroAnimation_58_#8:EStatCreatorIntroAnimation Enum
    OutroAnimation_61_#9
   )
   > n4 = Branch(Condition="Not Equal (Class)"(
     A:Object Class Reference=Get Class(Object=@n1) -> (ReturnValue:User Widget Class Reference)
     B:Object Class Reference=Resolve Soft Reference!(Input:User Widget Soft Class Reference=@n3.UserWidget_38_#14) -> (Output:User Widget Class Reference)
    ) -> (ReturnValue:bool=~) {is=title K2Node_PromotableOperator})
   Outputs!(True:Exec=[@n4.then, @n2.else, @n1.else], false:Exec=@n4.else)
  ]}
]
variables [
 IsActive {type=boolean}
 Level {type=integer}
 LoadedWidget {type=UserWidget, default=null}
 OnVisualStateUpdated {}
 OutroIsPlaying {type=boolean}
 Owning Stat Filter {type=FortMinigamePlayerDefinedStatFilter, default=null}
 Owning Stat Level Filter {type=FortMinigamePlayerDefinedStatFilter, default=null}
 Stat Bar Shown on HUD {type=E_StatManager_ShowOnHUD, default=Yes}
 StatComponent {type=FortMinigameStatsComponent, default=null}
 StatDetails {
  type=StatCreatorDetails
  default {
   name=""
   displayTime=0
   hUDPriority=0
   icon=null
   introAnimation=None
   invertClass=false
   invertTeam=false
   levelFilter=null
   maxLevel=0
   outroAnimation=None
   scope=Player
   showHUDInList=false
   userWidget=null
   valueFilter=null
   maxValues []
   selectedClass {classSlot=1, classType=ClassSlot}
   selectedTeam {teamIndex=1, teamType=TeamIndex}
   statColor=<rgba 0 0 0 0>
  }
 }
 Value {type=integer}
 ViewModel {type=MVVM_UEFN_Stat_C, default=null}
]
widgets {
 info {inheritedWidgetCount=0, parentClass=&UserWidget, rootWidgetClass=&CanvasPanel, widgetCount=1}
 widgets [Panel {widgetClassPath=&CanvasPanel, bIsVariable=true, widget=Panel}]
}
