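# StatContainerWidget is the HUD container that displays Stat Creator stats for the local player in a minigame.
# It takes stat details from the playspace component and filters them by disabled tags, team and class, re-checking on player, team or class changes.
# Valid stats queue for one-at-a-time async StatAnimationContainer creation, then land in a priority-ordered wrap list or a full-screen overlay.
# Visibility events remove inactive widgets and resize the list width.
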WidgetBlueprint /CRD_StatManager/UI/StatContainerWidget
graphs [
 AddWidgetToContainer {nodes [
   n1 = AddWidgetToContainer!() -> (then, Stat Device:Stat Creator Details Structure, Widget:User Widget Object Reference) {is=title K2Node_FunctionEntry}
   n2 = Break Stat Creator Details!(StatCreatorDetails:"Stat Creator Details Structure (by ref)"=@n1) -> (
    ValueFilter_53_#1:Fort Minigame Stat Filter Object Reference
    LevelFilter_54_#2:Fort Minigame Stat Filter Object Reference
    HUDPriority_19_#3:int
    ShowHUDInList_52_#4:bool=~
    IntroAnimation_58_#5:EStatCreatorIntroAnimation Enum
    OutroAnimation_61_#6:EStatCreatorOutroAnimation Enum
   ) {is=title K2Node_BreakStruct}
   > Set Filter(Filter:Fort Minigame Stat Filter Object Reference=@n2) -> (Output_Get:Fort Minigame Stat Filter Object Reference)
   > Set Widget Priority(Widget Priority:int=@n2.HUDPriority_19_#3) -> (Output_Get:int=0)
   > n3 = Cast To StatAnimationContainer!(execute, Object=@n1.Widget) -> (then, CastFailed, AsStat Animation Container:Stat Animation Container Object Reference) {is=title K2Node_DynamicCast}
   > Set TestWidget(TestWidget:Stat Animation Container Object Reference=@n3) -> (Output_Get:Stat Animation Container Object Reference)
   > n4 = Branch(Condition=@n2.ShowHUDInList_52_#4)
   n5 = Get StatList() -> (StatList:Wrap Box Object Reference)
   > RemoveChild(self=@n5, Content=Get TestWidget() -> (TestWidget:Stat Animation Container Object Reference))
   > n6 = Set Children(Children:Array of Widget Object References=GetAllChildren(self=@n5)) -> (Output_Get:Array of Widget Object References)
   > n7 = For Each Loop with Break!(Exec:Exec, Array:Array of Widget Object References=@n6, Break:Exec=@n13.then) -> (LoopBody:Exec, Array Element:Widget Object Reference, Array Index:int, Completed:Exec)
   n8 = Branch(execute=@n7.Completed, Condition=Get InsertedItem() -> (InsertedItem:bool))
   > ClearChildren(self=Get StatList())
   > n9 = For Each Loop(Array:Array of Widget Object References=Get Children() -> (Children:Array of Widget Object References)) -> (Array Element:Widget Object Reference)
   Add!(
    execute=[@n10.then, @n9.Completed]
    TargetMap:Map of Fort Minigame Stat Filter Object References to Stat Animation Container Object References=Get LoadedWidgets() -> (LoadedWidgets:Map of Fort Minigame Stat Filter Object References to Stat Animation Container Object References)
    Key:"Fort Minigame Stat Filter Object Reference (by ref)"=Get Filter() -> (Filter:Fort Minigame Stat Filter Object Reference)
    Value:"Stat Animation Container Object Reference (by ref)"=Get TestWidget()
   ) -> (then) {is=title K2Node_CallFunction}
   AddWidgetToList!(execute=@n9.LoopBody, self, WidgetToAdd:Widget Object Reference=@n9) -> (then)
   n10 = AddWidgetToList(execute=@n8.else, WidgetToAdd=Get TestWidget())
   n11 = Cast To StatAnimationContainer(execute=@n7.LoopBody, Object=@n7)
   n12 = Get StatDetails!(self:Stat Animation Container Object Reference=@n11) -> (StatDetails:Stat Creator Details Structure)
   > Branch(Condition="integer > integer"(
     A=Get Widget Priority() -> (Widget Priority:int=0)
     B=Break Stat Creator Details!(StatCreatorDetails=@n12) -> (HUDPriority_19_#3)
    ))
   > Insert!(
    execute
    TargetArray:Array of Widget Object References=Get Children()
    NewItem:"Widget Object Reference (by ref)"=Get Widget() -> (Widget:User Widget Object Reference)
    Index:int=@n7.Array Index
   ) -> (then) {is=title K2Node_CallArrayFunction}
   > n13 = Set InsertedItem(InsertedItem:bool=true) -> (Output_Get:bool)
   AddWidgetToOverlay!(execute=@n4.else, self, Content:Widget Object Reference=Get TestWidget()) -> (then)
   > Add(TargetMap=Get LoadedWidgets(), Key=Get Filter(), Value=Get TestWidget())
  ]}
 AddWidgetToList {nodes [
   n1 = AddWidgetToList!() -> (then, WidgetToAdd:Widget Object Reference) {is=title K2Node_FunctionEntry}
   > n2 = AddChildToWrapBox!(execute, self:Wrap Box Object Reference=Get StatList(), Content:Widget Object Reference=@n1) -> (then, ReturnValue:Wrap Box Slot Object Reference)
   n3 = Get WidgetPadding() -> (WidgetPadding:float)
   > SetPadding!(
    execute
    self:Wrap Box Slot Object Reference=@n2
    InPadding_Left:float=0
    InPadding_Top:float=@n3
    InPadding_Right:float=@n3
    InPadding_Bottom:float=0
   ) -> (then)
   > n4 = Cast To StatAnimationContainer(Object=@n1)
   > Bind Event to On Visual State Updated(self:Stat Animation Container Object Reference=@n4, Delegate=Create Event()) {is=title K2Node_AddDelegate}
  ]}
 AddWidgetToOverlay {nodes [
   n1 = AddWidgetToOverlay!() -> (then, Content:Widget Object Reference) {is=title K2Node_FunctionEntry}
   > RemoveChild(self=Get StatOverlayParent() -> (StatOverlayParent:Overlay Object Reference), Content=@n1)
   > n2 = AddChild(self=Get StatOverlayParent(), Content=@n1)
   > n3 = Cast To OverlaySlot!(execute, Object=@n2) -> (then, CastFailed, AsOverlay Slot:Overlay Slot Object Reference) {is=title K2Node_DynamicCast}
   > SetVerticalAlignment(self=@n3, InVerticalAlignment=VAlign_Fill)
   > SetHorizontalAlignment(self=@n3, InHorizontalAlignment=HAlign_Fill)
  ]}
 Bind Minigame {nodes [
   n1 = Bind Minigame!() -> (then, Bind:bool=~) {is=title K2Node_FunctionEntry}
   n2 = GetOwningPlayer()
   > n3 = Is Valid(InputObject=@n2)
   FortShippingLog!(
    execute=[@n5.Is Not Valid, @n3.Is Not Valid]
    WorldContextObject:Object Reference
    InString:String=SCQX: SCW - Could not bind minigame
    bLogAsWarning:bool=false
   ) -> (then)
   > Attempt To Rebind Minigame()
   n4 = GetMinigame!(ContextActor:Actor Object Reference=@n2) -> (ReturnValue:Fort Minigame Object Reference)
   n5 = Is Valid(exec=@n3.Is Valid, InputObject=@n4)
   > n6 = Set CachedMinigame(CachedMinigame:Fort Minigame Object Reference=@n4) -> (Output_Get:Fort Minigame Object Reference)
   > n7 = Branch(Condition=@n1)
   n8 = Create Event()
   > Bind Event to On Any Local Player Changed Class(self:Fort Minigame Object Reference=@n6, Delegate=@n8) {is=title K2Node_AddDelegate}
   n9 = Get PlayerState!(self:Controller Object Reference=GetOwningPlayer()) -> (PlayerState:Player State Object Reference)
   > n10 = Cast To FortPlayerStateAthena!(execute, Object=@n9) -> (then, CastFailed, AsFort Player State Athena:Fort Player State Athena Object Reference) {is=title K2Node_DynamicCast}
   n11 = Create Event()
   > Bind Event to On Team Index Changed Delegate(self:Fort Player State Athena Object Reference=@n10, Delegate=@n11) {is=title K2Node_AddDelegate}
   n12 = Create Event()
   > Bind Event to On Finished Adding Minigame Player Delegate(self:Fort Minigame Object Reference=@n6, Delegate=@n12) {is=title K2Node_AddDelegate}
   Unbind Event from On Any Local Player Changed Class(execute=@n7.else, self:Fort Minigame Object Reference=@n6, Delegate=@n8) {is=title K2Node_RemoveDelegate}
   > n13 = Cast To FortPlayerStateAthena(Object=@n9)
   > Unbind Event from On Team Index Changed Delegate(self:Fort Player State Athena Object Reference=@n13, Delegate=@n11) {is=title K2Node_RemoveDelegate}
   > Unbind Event from On Finished Adding Minigame Player Delegate(self:Fort Minigame Object Reference=@n6, Delegate=@n12) {is=title K2Node_RemoveDelegate}
  ]}
 EventGraph {nodes [
   OnLocalPlayerTeamOrClassChanged() {is=title K2Node_CustomEvent}
   > InitializeAllStats()
   n1 = Stat Visibility Updated!() -> (
    OutputDelegate
    then
    Widget:Stat Animation Container Object Reference
    Width:double=~
    TrackedStat:Fort Minigame Player Defined Stat Filter Object Reference
    Active:bool=~
   ) {is=title K2Node_CustomEvent}
   > Add(
    TargetMap:"Map of User Widget Object References to Float (double-precision)s"=Get WidgetSizes() -> (WidgetSizes:"Map of User Widget Object References to Float (double-precision)s")
    Key:"User Widget Object Reference (by ref)"=@n1.Widget
    Value:"Float (double-precision) (by ref)"=@n1.Width
   )
   > n2 = Branch(Condition=@n1.Active)
   > DelayUntilNextTick!(execute) -> (then)
   > UpdateContainerSize()
   Remove Stat!(
    execute=@n2.else
    self
    User Widget:Stat Animation Container Object Reference=@n1.Widget
    Stat Filter:Fort Minigame Player Defined Stat Filter Object Reference=@n1.TrackedStat
   ) -> (then)
   n3 = On Disabled Set Changed!() -> (OutputDelegate, then, InactiveStats:Set of Gameplay Tag Structures) {is=title K2Node_CustomEvent}
   > Set List of Disabled Stats(List of Disabled Stats:Set of Gameplay Tag Structures=@n3.InactiveStats) -> (Output_Get:Set of Gameplay Tag Structures)
   > InitializeAllStats()
   n4 = Client Details Updated!() -> (OutputDelegate, then, New Details:Map of Fort Minigame Stat Filter Object References to Stat Creator Details Structures) {is=title K2Node_CustomEvent}
   > Set AllStats(AllStats:Map of Fort Minigame Stat Filter Object References to Stat Creator Details Structures=@n4.New Details) -> (Output_Get:Map of Fort Minigame Stat Filter Object References to Stat Creator Details Structures)
   > InitializeAllStats()
   n5 = Add Filter to Queue!() -> (OutputDelegate, then, StatFilter:Fort Minigame Stat Filter Object Reference) {is=title K2Node_CustomEvent}
   > n6 = Add(
    TargetArray:Array of Fort Minigame Stat Filter Object References=Get QueuedFilters() -> (QueuedFilters:Array of Fort Minigame Stat Filter Object References)
    NewItem:"Fort Minigame Stat Filter Object Reference (by ref)"=@n5.StatFilter
   ) {is=title K2Node_CallArrayFunction}
   > n7 = Branch(Condition="Equal (Integer)"(A=@n6))
   Branch(
    execute=[@n7.then, @n15.then]
    Condition=Is Valid Index!(TargetArray:Array of Fort Minigame Stat Filter Object References=Get QueuedFilters(), IndexToTest:int) -> (ReturnValue:bool)
   )
   n8 = "Get (a copy)"(Array:Array of Fort Minigame Stat Filter Object References=Get QueuedFilters()) -> (Output:Fort Minigame Stat Filter Object Reference)
   > Set LoadingDetails(LoadingDetails:Stat Creator Details Structure=Find(
     TargetMap:Map of Fort Minigame Stat Filter Object References to Stat Creator Details Structures=Get AllStats() -> (AllStats:Map of Fort Minigame Stat Filter Object References to Stat Creator Details Structures)
     Key:"Fort Minigame Stat Filter Object Reference (by ref)"=@n8
    ) -> (Value:Stat Creator Details Structure)) -> (Output_Get:Stat Creator Details Structure)
   n9 = Find(TargetMap:Map of Fort Minigame Stat Filter Object References to Stat Animation Container Object References=Get LoadedWidgets(), Key=@n8) -> (Value:Stat Animation Container Object Reference)
   > n10 = Branch(Condition=NOT Boolean(A=@n9.ReturnValue))
   > n11 = CreateWidgetAsync!(
    execute:Exec
    UserWidgetSoftClass:User Widget Soft Class Reference=&/StatAnimationContainer.StatAnimationContainer_C
    OwningPlayer:Player Controller Object Reference
    bSuspendInputUntilComplete:bool=false
   ) -> (then:Exec, OnComplete:Exec, OnFailure:Exec, UserWidget:User Widget Object Reference) {is=title K2Node_AsyncAction}
   n12 = Branch(execute=@n11.OnComplete, Condition=Is Valid Index(TargetArray=Get QueuedFilters()))
   > AddWidgetToContainer!(
    execute
    self
    Stat Device:Stat Creator Details Structure=Get LoadingDetails() -> (LoadingDetails:Stat Creator Details Structure)
    Widget:User Widget Object Reference=@n11
   ) -> (then) {is=title K2Node_CallFunction}
   > n13 = Cast To StatAnimationContainer(Object=@n11)
   > n14 = Initialize From StatDetails!(execute, self:Stat Animation Container Object Reference=@n13, Details:Stat Creator Details Structure=Get LoadingDetails()) -> (then)
   Remove Index!(
    execute=[@n14.then, @n16.then, @n12.else]
    TargetArray:Array of Fort Minigame Stat Filter Object References=Get QueuedFilters()
    IndexToRemove:int=0
   ) -> (then)
   > n15 = Set LoadingDetails()
   n16 = Initialize From StatDetails(execute=@n10.else, self=@n9, Details=Get LoadingDetails())
   Attempt To Rebind Minigame() {is=title K2Node_CustomEvent}
   > Delay(Duration=2)
   > Bind Minigame!(execute, self, Bind:bool=true) -> (then) {is=title K2Node_CallFunction}
   Event Construct()
   > Bind Minigame(Bind=true)
   > n17 = GetPlaySpaceComponent!(execute, self) -> (then, StatDetailsComponent:Playspace Component Player Stat Details Object Reference)
   > Bind Event to On Client Enabled State Changed(self:Playspace Component Player Stat Details Object Reference=@n17, Delegate=@n3) {is=title K2Node_AddDelegate}
   > Bind Event to On Client Details Updated(self:Playspace Component Player Stat Details Object Reference=@n17, Delegate=@n4) {is=title K2Node_AddDelegate}
   > n18 = Get Stat Details!(execute, self:Playspace Component Player Stat Details Object Reference=@n17) -> (
    then
    StatDetails:Map of Fort Minigame Stat Filter Object References to Stat Creator Details Structures
    DisabledStats:Set of Gameplay Tag Structures
   ) {is=title K2Node_CallFunction}
   > Set AllStats(AllStats=@n18)
   > Set List of Disabled Stats(List of Disabled Stats=@n18.DisabledStats)
   > InitializeAllStats()
  ]}
 GetPlaySpaceComponent {nodes [
   GetPlaySpaceComponent() {is=title K2Node_FunctionEntry}
   > n1 = Get!(execute:Exec) -> (then:Exec, else:Exec, StatDetailsComponent:Playspace Component Player Stat Details Object Reference)
   > Return Node!(execute, StatDetailsComponent:Playspace Component Player Stat Details Object Reference=@n1)
   n2 = GetFirstLocalFortPlayerController!(bIncludeNonSpectators:bool=true, bIncludeSpectators:bool=false) -> (ReturnValue:Fort Player Controller Object Reference)
   n3 = GetPlayspaceForActor!(execute=@n1.else, WorldContextObject:Object Reference=@n2, InActor:Actor Object Reference=@n2) -> (then, ReturnValue:Playspace Object Reference)
   > n4 = Set StatDetailsComponent(StatDetailsComponent:Playspace Component Player Stat Details Object Reference=GetComponentByClass!(
     self:Actor Object Reference=@n3
     ComponentClass:Actor Component Class Reference=&/PlayspaceComponent_PlayerStatDetails.PlayspaceComponent_PlayerStatDetails_C
    ) -> (ReturnValue:Playspace Component Player Stat Details Object Reference)) -> (Output_Get:Playspace Component Player Stat Details Object Reference)
   > Return Node(StatDetailsComponent=@n4)
  ]}
 InitializeAllStats {nodes [
   InitializeAllStats() {is=title K2Node_FunctionEntry}
   > n1 = Keys!(execute, TargetMap:Map of Fort Minigame Stat Filter Object References to Stat Creator Details Structures=Get AllStats()) -> (then, Keys:Array of Fort Minigame Stat Filter Object References)
   > n2 = For Each Loop(Array:Array of Fort Minigame Stat Filter Object References=@n1) -> (Array Element:Fort Minigame Stat Filter Object Reference)
   > Set CheckStat(CheckStat:Fort Minigame Stat Filter Object Reference=@n2) -> (Output_Get:Fort Minigame Stat Filter Object Reference)
   > n3 = Set CheckingDetails(CheckingDetails:Stat Creator Details Structure=Find(
     TargetMap:Map of Fort Minigame Stat Filter Object References to Stat Creator Details Structures=Get AllStats()
     Key=Get CheckStat() -> (CheckStat:Fort Minigame Stat Filter Object Reference)
    ) -> (Value:Stat Creator Details Structure)) -> (Output_Get:Stat Creator Details Structure)
   > n4 = IsValidForLocalPlayer!(execute, self, Stat Details:Stat Creator Details Structure=@n3) -> (then, Success:bool)
   > n5 = Branch(Condition=@n4)
   > Add Filter to Queue!(execute, self, StatFilter:Fort Minigame Stat Filter Object Reference=Get CheckStat()) -> (then) {is=title K2Node_CallFunction}
   n6 = Get CheckStat()
   Add(
    execute=@n5.else
    TargetMap:Map of Fort Minigame Stat Filter Object References to Stat Creator Details Structures=Get InvalidStats() -> (InvalidStats:Map of Fort Minigame Stat Filter Object References to Stat Creator Details Structures)
    Key:"Fort Minigame Stat Filter Object Reference (by ref)"=@n6
    Value:"Stat Creator Details Structure (by ref)"=Get CheckingDetails() -> (CheckingDetails:Stat Creator Details Structure)
   ) {is=title K2Node_CallFunction}
   > Remove Item!(
    execute
    TargetArray:Array of Fort Minigame Stat Filter Object References=Get QueuedFilters()
    Item:"Fort Minigame Stat Filter Object Reference (by ref)"=@n6
   ) -> (then, ReturnValue:bool)
   n7 = Find(
    TargetMap:Map of Fort Minigame Stat Filter Object References to Stat Animation Container Object References=Get LoadedWidgets()
    Key=Get CheckStat()
   ) -> (Value:Stat Animation Container Object Reference)
   > Branch(Condition=@n7.ReturnValue)
   > Force Close!(execute, self:Stat Animation Container Object Reference=@n7, StillActive:bool=false) -> (then)
  ]}
 Is Stat Filter Disabled {nodes [
   n1 = Is Stat Filter Disabled!() -> (then, StatFilter:Fort Minigame Stat Filter Object Reference) {is=title K2Node_FunctionEntry}
   > n2 = To Array!(execute, A:Set of Gameplay Tag Structures=Get List of Disabled Stats() -> (List of Disabled Stats:Set of Gameplay Tag Structures)) -> (then, Result:Array of Gameplay Tag Structures)
   > n3 = For Each Loop(Array:Array of Gameplay Tag Structures=@n2) -> (Array Element:Gameplay Tag Structure)
   Return Node!(execute=@n3.Completed, ReturnValue:bool=false)
   n4 = MakeLiteralGameplayTagContainer!(Value:Gameplay Tag Container Structure) -> (ReturnValue:Gameplay Tag Container Structure)
   > Branch(Condition=Matches!(
     self:Fort Minigame Stat Filter Object Reference=@n1
     InStatEvent:EFortQuestObjectiveStatEvent Enum=MinigameDynamicEvent
     InTargetObject:Object Reference
     InPlayerController:Controller Object Reference
     InTargetTags:"Gameplay Tag Container Structure (by ref)"=MakeGameplayTagContainerFromTag!(SingleTag:Gameplay Tag Structure=@n3) -> (ReturnValue:Gameplay Tag Container Structure)
     InSourceTags:"Gameplay Tag Container Structure (by ref)"=@n4
     InContextTags:"Gameplay Tag Container Structure (by ref)"=@n4
    ) -> (ReturnValue:bool))
   > Return Node(ReturnValue=true)
  ]}
 IsValidForLocalPlayer {nodes [
   n1 = IsValidForLocalPlayer!() -> (then, Stat Details:Stat Creator Details Structure) {is=title K2Node_FunctionEntry}
   > n2 = Sequence!(execute) -> (then_0, then_1, then_2:Exec, then_3:Exec)
   > Branch(Condition=Is Stat Filter Disabled!(StatFilter:Fort Minigame Stat Filter Object Reference=Break Stat Creator Details!(StatCreatorDetails=@n1) -> (
      ValueFilter_53_#1:Fort Minigame Stat Filter Object Reference
      LevelFilter_54_#2:Fort Minigame Stat Filter Object Reference
      Name_41_#7:Text
      Icon_45_#8:Texture 2D Soft Object Reference
      Scope_50_#9:E_StatManager_Scope Enum
      StatColor_7_#10:Linear Color Structure
      MaxValues_37_#11:Array of Integers
      MaxLevel_12_#12:int
      UserWidget_38_#13:User Widget Soft Class Reference
      HUDPriority_19_#3
      ShowHUDInList_52_#4:bool=~
      DisplayTime_21_#14:double=~
      SelectedTeam_26_#15:Creative Team Option Structure
      InvertTeam_32_#16:bool=~
      SelectedClass_34_#17:Creative Class Option Structure
      InvertClass_33_#18:bool=~
      IntroAnimation_58_#5:EStatCreatorIntroAnimation Enum
      OutroAnimation_61_#6:EStatCreatorOutroAnimation Enum
     )) -> (ReturnValue:bool) {is=title K2Node_CallFunction})
   > Return Node!(execute, Success:bool=false)
   n3 = Break Stat Creator Details!(StatCreatorDetails=Get Stat Details() -> (Stat Details:Stat Creator Details Structure) {is=title K2Node_VariableGet}) -> (
    ValueFilter_53_#1
    LevelFilter_54_#2
    Scope_50_#9
    ShowHUDInList_52_#4=~
    SelectedTeam_26_#15
    InvertTeam_32_#16=~
    IntroAnimation_58_#5
    OutroAnimation_61_#6
   )
   n4 = Branch(
    execute=@n2.then_1
    Condition=IsTeamType!(
     TeamOption:"Creative Team Option Structure (by ref)"=@n3.SelectedTeam_26_#15
     TeamType:ECreativeTeamType Enum=Any
    ) -> (ReturnValue:bool)
   )
   > n5 = Set IsValidTeam(IsValidTeam:bool=NOT Boolean(A=@n3.InvertTeam_32_#16)) -> (Output_Get:bool)
   Branch(execute=[@n5.then, @n8.then], Condition=NOT Boolean(A=Get IsValidTeam() -> (IsValidTeam:bool)))
   > Return Node(Success=false)
   n6 = GetActorPvPTeamIndex!(Actor:Actor Object Reference=GetOwningPlayer()) -> (TeamIndex:Byte=0, ReturnValue:bool)
   n7 = Byte - Byte(
    A:Byte=ConvertToTeam!(TeamOption:"Creative Team Option Structure (by ref)"=@n3.SelectedTeam_26_#15) -> (ReturnValue:EFortTeam Enum=Spectator)
    B:Byte=3
   ) -> (ReturnValue:Byte) {is=title K2Node_PromotableOperator}
   n8 = Set IsValidTeam(
    execute=@n4.else
    IsValidTeam="Not Equal (Boolean)"(
     A=@n3.InvertTeam_32_#16
     B="Equal (Byte)"(A:Byte=@n7, B:Byte=@n6) -> (ReturnValue:bool=~) {is=title K2Node_PromotableOperator}
    ) -> (ReturnValue=~)
   )
   n9 = Get Stat Details()
   n10 = Branch(
    execute=@n2.then_2
    Condition=IsClassType!(
     ClassOption:"Creative Class Option Structure (by ref)"=Break Stat Creator Details!(StatCreatorDetails=@n9) -> (SelectedClass_34_#17:Creative Class Option Structure)
     ClassType:ECreativeClassType Enum=Any
    ) -> (ReturnValue:bool)
   )
   n11 = Break Stat Creator Details!(StatCreatorDetails=Get Stat Details()) -> (SelectedClass_34_#17, InvertClass_33_#18:bool=~)
   > n12 = Set IsValidClass(IsValidClass:bool=NOT Boolean(A=@n11.InvertClass_33_#18)) -> (Output_Get:bool)
   Return Node(execute=[@n18.then, @n12.then], Success=Get IsValidClass() -> (IsValidClass:bool))
   n13 = Cast To FortPlayerState!(execute=@n10.else, Object=Get PlayerState(self=GetOwningPlayer())) -> (then, CastFailed, AsFort Player State:Fort Player State Object Reference) {is=title K2Node_DynamicCast}
   n14 = Get CachedMinigame() -> (CachedMinigame:Fort Minigame Object Reference)
   > Is Valid(InputObject=@n14)
   n15 = GetFNEClassComponent!(self:Fort Minigame Object Reference=@n14) -> (ReturnValue:FNEClass Component Object Reference)
   > Is Valid(InputObject=@n15)
   n16 = ConvertToClassSlotIndex!(ClassOption:"Creative Class Option Structure (by ref)"=@n11) -> (ReturnValue:Byte=0)
   n17 = GetClassSlotForPlayer!(self:FNEClass Component Object Reference=@n15, PlayerState:Fort Player State Object Reference=@n13, bUseDeathInfo:bool=false) -> (ReturnValue:Byte=0)
   > n18 = Set IsValidClass(IsValidClass="Not Equal (Boolean)"(A="Equal (Byte)"(A=@n17, B=@n16) -> (ReturnValue=~), B=@n11.InvertClass_33_#18) -> (ReturnValue=~))
   Return Node(execute=@n2.then_3, Success=false)
  ]}
 OnFinishedAddingMinigamePlayer {nodes [
   n1 = OnFinishedAddingMinigamePlayer!() -> (then, UniqueNetId:Unique Net Id Structure, bIsLocalPlayer:bool=~, FortPlayerState:Fort Player State Object Reference) {is=title K2Node_FunctionEntry}
   > Branch(Condition="Equal (Object)"(
     A=GetPlayerController!(self:Player State Object Reference=@n1.FortPlayerState) -> (ReturnValue:Player Controller Object Reference)
     B=GetOwningPlayer()
    ))
   > InitializeAllStats() {is=title K2Node_CallFunction}
  ]}
 Remove Stat {nodes [
   n1 = Remove Stat!() -> (then, User Widget:Stat Animation Container Object Reference, Stat Filter:Fort Minigame Player Defined Stat Filter Object Reference) {is=title K2Node_FunctionEntry}
   n2 = Get LoadingDetails()
   > n3 = Branch(Condition="Equal (Object)"(
     A=Break Stat Creator Details!(StatCreatorDetails=@n2) -> (
      ValueFilter_53_#1:Fort Minigame Stat Filter Object Reference
      LevelFilter_54_#2:Fort Minigame Stat Filter Object Reference
      Scope_50_#9:E_StatManager_Scope Enum
      ShowHUDInList_52_#4:bool=~
      IntroAnimation_58_#5:EStatCreatorIntroAnimation Enum
      OutroAnimation_61_#6:EStatCreatorOutroAnimation Enum
     )
     B=@n1.Stat Filter
    ))
   > n4 = Set LoadingDetails()
   Branch(execute=[@n3.else, @n4.then], Condition=Find(TargetMap=Get LoadedWidgets(), Key=@n1.Stat Filter).ReturnValue)
   n5 = Get User Widget() -> (User Widget:Stat Animation Container Object Reference)
   > RemoveFromParent(self=@n5)
   > Unbind Event from On Visual State Updated(self:Stat Animation Container Object Reference=@n5, Delegate=Create Event()) {is=title K2Node_RemoveDelegate}
   > Remove!(
    execute
    TargetMap:"Map of User Widget Object References to Float (double-precision)s"=Get WidgetSizes()
    Key:"User Widget Object Reference (by ref)"=@n5
   ) -> (then, ReturnValue:bool)
   n6 = Get Stat Filter() -> (Stat Filter:Fort Minigame Player Defined Stat Filter Object Reference)
   > Add(
    TargetMap=Get InvalidStats()
    Key=@n6
    Value=Find(TargetMap:Map of Fort Minigame Stat Filter Object References to Stat Creator Details Structures=Get AllStats(), Key=@n6) -> (Value:Stat Creator Details Structure)
   )
   > Remove(
    TargetMap:Map of Fort Minigame Stat Filter Object References to Stat Animation Container Object References=Get LoadedWidgets()
    Key:"Fort Minigame Stat Filter Object Reference (by ref)"=Get Stat Filter()
   )
  ]}
 UpdateContainerSize {nodes [
   UpdateContainerSize() {is=title K2Node_FunctionEntry}
   > n1 = Values!(execute, TargetMap:"Map of User Widget Object References to Float (double-precision)s"=Get WidgetSizes()) -> (then, Values:"Array of Float (double-precision)s")
   > Set MaxChildWidth(MaxChildWidth:double=MaxOfFloatArray!(FloatArray:"Array of Float (single-precision)s"=@n1) -> (IndexOfMaxValue:int=0, MaxValue:float).MaxValue) -> (Output_Get:double)
   n2 = "Max (Float)"(A=Get MaxChildWidth() -> (MaxChildWidth:double), B=Get MinWidthLimit() -> (MinWidthLimit:double))
   > n3 = Set WidthToSet(WidthToSet:double=@n2) -> (Output_Get:double)
   n4 = float + float(A=@n3, B=@n2)
   > n5 = While Loop!(
    execute:Exec
    Condition:bool="float < float"(A=@n4, B=Get!() -> (ContainerWidthMinMax_X:double, ContainerWidthMinMax_Y:double).ContainerWidthMinMax_Y)
   ) -> (LoopBody:Exec, Completed:Exec) {is=title K2Node_MacroInstance}
   SetWidthOverride(
    execute=@n5.Completed
    self=Get StatListParent() -> (StatListParent:Size Box Object Reference)
    InWidthOverride=Get WidthToSet() -> (WidthToSet:double)
   )
   Set WidthToSet(execute=@n5.LoopBody, WidthToSet=@n4)
  ]}
]
variables [
 AllStats {default {}}
 CachedMinigame {type=FortMinigame, default=null}
 ContainerWidthMinMax {type=Vector2D, default=<320 500>}
 InvalidStats {default {}}
 List of Disabled Stats {type="array<GameplayTag>", default []}
 LoadedWidgets {default {}}
 LoadingDetails {
  type=StatCreatorDetails
  default {
   name=""
   displayTime=0
   hUDPriority=0
   icon=null
   introAnimation=None
   invertClass=false
   invertTeam=false
   levelFilter=null
   maxLevel=0
   outroAnimation=None
   scope=Player
   showHUDInList=false
   userWidget=null
   valueFilter=null
   maxValues []
   selectedClass {classSlot=1, classType=ClassSlot}
   selectedTeam {teamIndex=1, teamType=TeamIndex}
   statColor=<rgba 0 0 0 0>
  }
 }
 QueuedFilters {type="array<FortMinigameStatFilter>", default []}
 Register Ref {type=CreativeIndexedObjectRegisterComp_C, default=null}
 StatDetailsComponent {type=PlayspaceComponent_PlayerStatDetails_C, default=null}
 WidgetPadding {type=number, default=5}
 WidgetSizes {default {}}
]
widgets {
 info {inheritedWidgetCount=0, parentClass=&UserWidget, rootWidgetClass=&CanvasPanel, widgetCount=4}
 widgets [
  CanvasPanel_1 {widgetClassPath=&CanvasPanel, widget=CanvasPanel_1}
  StatListParent {
   widgetClassPath=&SizeBox
   bIsVariable=true
   parent=CanvasPanel_1
   widget=StatListParent
   properties {
    bOverride_WidthOverride=true
    clipping=ClipToBounds
    heightOverride=500
    maxDesiredWidth=320
    minDesiredHeight=500
    minDesiredWidth=320
    widthOverride=320
   }
   slot {
    class=CanvasPanelSlot
    properties {bAutoSize=true, layoutData {alignment=<0 1>, anchors {maximum=<0 1>, minimum=<0 1>}, offsets=<ltrb 20 -200 320 0>}}
   }
  }
  StatList {
   widgetClassPath=&WrapBox
   bIsVariable=true
   parent=StatListParent
   widget=StatList
   properties {bExplicitWrapSize=true, orientation=Orient_Vertical}
   slot {class=SizeBoxSlot}
  }
  StatOverlayParent {
   widgetClassPath=&Overlay
   bIsVariable=true
   parent=CanvasPanel_1
   widget=StatOverlayParent
   slot {class=CanvasPanelSlot, properties {layoutData {alignment=<0.5 0.5>, anchors {maximum=<1 1>, minimum=<0 0>}, offsets=<ltrb 0 0 0 0>}}}
  }
 ]
}
