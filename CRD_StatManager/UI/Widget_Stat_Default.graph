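# Default UMG stat row for a UEFN Stat Creator device, showing icon, name, progress bar, level and value/max or "Max" text.
# MVVM helper functions convert the view model's level to text, tint the maximum-value texts and set the ActualProgress fill.
# Each fill update starts UpdateProgressFillToActual, which eases the overlaid AnimatedProgress bar toward ActualProgress with FInterpTo.
# It loops through DelayUntilNextTick rather than Tick, snapping once nearly equal.
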WidgetBlueprint /CRD_StatManager/UI/Widget_Stat_Default
graphs [
 EventGraph {nodes [
   n1 = UpdateProgressFillToActual() {is=title K2Node_CustomEvent}
   n2 = Get Percent!(self:Progress Bar Object Reference=Get ActualProgress() -> (ActualProgress:Progress Bar Object Reference)) -> (Percent:float)
   n3 = Get Percent(self=Get AnimatedProgress() -> (AnimatedProgress:Progress Bar Object Reference))
   n4 = Branch(execute=[@n1.then, @n8.then], Condition=NOT Boolean(A="Nearly Equal (Float)"(A=@n3, B=@n2, ErrorTolerance=0.0001)))
   n5 = Get Percent(self=Get ActualProgress())
   n6 = Get AnimatedProgress()
   n7 = Get AnimatedProgress()
   > SetPercent!(
    execute
    self:Progress Bar Object Reference=@n7
    InPercent:float=FInterpTo(
     Current=Get Percent(self=@n6)
     Target=@n5
     DeltaTime=GetWorldDeltaSeconds!() -> (ReturnValue:double)
     InterpSpeed=Get ProgressInterpSpeed() -> (ProgressInterpSpeed:double)
    )
   ) -> (then)
   > n8 = DelayUntilNextTick!(execute) -> (then)
   SetPercent(execute=@n4.else, self=@n7, InPercent=@n5)
  ]}
 MVVM Color Change {nodes [
   n1 = MVVM Color Change!() -> (then, Color:Linear Color Structure) {is=title K2Node_FunctionEntry}
   > SetColorAndOpacity(
    self=[
     Get MaximumText() -> (MaximumText:Common Text Object Reference)
     Get ValueMax_Number() -> (ValueMax_Number:Common Text Object Reference)
     Get ValueSeparator() -> (ValueSeparator:Common Text Object Reference)
    ]
    InColorAndOpacity=Make Slate Color(SpecifiedColor=@n1)
   )
  ]}
 MVVM_UpdateFillPercent {nodes [
   n1 = MVVM_UpdateFillPercent!() -> (then, %:double=~) {is=title K2Node_FunctionEntry}
   > SetPercent(self=Get ActualProgress(), InPercent=@n1)
   > UpdateProgressFillToActual() {is=title K2Node_CallFunction}
  ]}
 __232426eb-4127-378a-8c75-6389384dac29_SourceToDest {nodes [
   __232426eb-4127-378a-8c75-6389384dac29_SourceToDest() {is=title K2Node_FunctionEntry}
   n1 = Get Level!(self:Device - Stat View Model Object Reference=Get MVVM_UEFN_Stat() -> (MVVM_UEFN_Stat:Device - Stat View Model Object Reference)) -> (Level:Integer64=0)
   > Return Node!(
    execute
    ReturnValue:Text="To Text (Integer)"(Value="To Integer (Integer64)"!(InInt:Integer64=@n1) -> (ReturnValue:int=0), bUseGrouping=false)
   )
  ]}
]
variables [
 ListObject {type=Device_StatCreator_C, default=null}
 ProgressInterpSpeed {type=number, default=1}
]
widgets {
 info {inheritedWidgetCount=0, parentClass=&UserWidget, rootWidgetClass=&Overlay, widgetCount=25}
 widgets [
  RootOverlay {widgetClassPath=&Overlay, widget=RootOverlay}
  ParentSizeBox {
   widgetClassPath=&SizeBox
   parent=RootOverlay
   widget=ParentSizeBox
   properties {bOverride_WidthOverride=true, heightOverride=71.0229, minDesiredWidth=250, widthOverride=250}
   slot {class=OverlaySlot, properties {verticalAlignment=VAlign_Center}}
  }
  AllContainer {widgetClassPath=&Overlay, parent=ParentSizeBox, widget=AllContainer, slot {class=SizeBoxSlot}}
  BackgroundBorder {
   widgetClassPath=&CommonBorder
   parent=AllContainer
   widget=BackgroundBorder
   properties {
    background {
     bIsDynamicallyLoaded=false
     drawAs=Box
     imageType=NoImage
     mirroring=NoMirror
     resourceName=None
     resourceObject=/Game/Athena/HUD/MiniMap/Art/M_UI_Quest_Container_S19_Fill
     tiling=NoTile
     imageSize=<64 64>
     margin=<ltrb 0.5 0.5 0.5 0.5>
     outlineSettings {
      bUseBrushTransparency=false
      roundingType=HalfHeightRadius
      width=0
      color {colorUseRule=UseColor_Specified, specifiedColor=<rgba 0 0 0 0>}
      cornerRadii=<0 0 0 0>
     }
     tintColor {colorUseRule=UseColor_Specified, specifiedColor=<rgba 1 1 1 1>}
     uVRegion {bIsValid=false, max=<0 0>, min=<0 0>}
    }
    padding=<ltrb 0 0 0 0>
   }
   slot {class=OverlaySlot, properties {horizontalAlignment=HAlign_Fill, verticalAlignment=VAlign_Fill}}
  }
  AllDetailsBox {
   widgetClassPath=&VerticalBox
   parent=AllContainer
   widget=AllDetailsBox
   slot {class=OverlaySlot, properties {horizontalAlignment=HAlign_Fill, verticalAlignment=VAlign_Center, padding=<ltrb 60 4 10 4>}}
  }
  StatName {
   widgetClassPath=&CommonTextBlock
   parent=AllDetailsBox
   widget=StatName
   properties {
    text=Text Block
    colorAndOpacity {colorUseRule=UseColor_Specified, specifiedColor=<rgba 0.2195 0.6584 1 1>}
    font {
     bForceMonospaced=false
     bMaterialIsStencil=false
     fontMaterial=null
     fontObject=&BurbankSmall
     letterSpacing=0
     monospacedWidth=1
     size=12
     skewAmount=0
     typefaceFontName=Bold
     outlineSettings {
      bApplyOutlineToDropShadows=false
      bMiteredCorners=false
      bSeparateFillAlpha=false
      outlineMaterial=null
      outlineSize=0
      outlineColor=<rgba 0 0 0 1>
     }
    }
    shadowOffset=<0 0>
   }
   slot {class=VerticalBoxSlot, properties {horizontalAlignment=HAlign_Left, verticalAlignment=VAlign_Center}}
  }
  ProgressSizeBox {
   widgetClassPath=&SizeBox
   bIsVariable=true
   parent=AllDetailsBox
   widget=ProgressSizeBox
   properties {bOverride_HeightOverride=true, heightOverride=16}
   slot {class=VerticalBoxSlot, properties {verticalAlignment=VAlign_Center, size {sizeRule=Fill, value=1}}}
  }
  ProgressBarBox {widgetClassPath=&CanvasPanel, parent=ProgressSizeBox, widget=ProgressBarBox, slot {class=SizeBoxSlot}}
  ActualProgress {
   widgetClassPath=/Script/UMG.ProgressBar
   bIsVariable=true
   parent=ProgressBarBox
   widget=ActualProgress
   properties {
    visibility=HitTestInvisible
    widgetStyle {
     enableFillAnimation=true
     backgroundImage {
      bIsDynamicallyLoaded=false
      drawAs=Box
      imageType=FullColor
      mirroring=NoMirror
      resourceName=../../../Engine/Content/Slate/Common/ProgressBar_Background.png
      resourceObject=/FortUILibrary/Foundation/Art_ToolKit/Tools/MI_UI_Circle_ProgressBar
      tiling=NoTile
      imageSize=<8 8>
      margin=<ltrb 0.5 0.5 0.5 0.5>
      outlineSettings {
       bUseBrushTransparency=false
       roundingType=HalfHeightRadius
       width=0
       color {colorUseRule=UseColor_Specified, specifiedColor=<rgba 0 0 0 0>}
       cornerRadii=<0 0 0 0>
      }
      tintColor {colorUseRule=UseColor_Specified, specifiedColor=<rgba 0 0 0 0.5>}
      uVRegion {bIsValid=false, max=<0 0>, min=<0 0>}
     }
     fillImage {
      bIsDynamicallyLoaded=false
      drawAs=Box
      imageType=FullColor
      mirroring=NoMirror
      resourceName=../../../Engine/Content/Slate/Common/ProgressBar_Fill.png
      resourceObject=null
      tiling=NoTile
      imageSize=<8 8>
      margin=<ltrb 0.5 0.5 0.5 0.5>
      outlineSettings {
       bUseBrushTransparency=false
       roundingType=HalfHeightRadius
       width=0
       color {colorUseRule=UseColor_Specified, specifiedColor=<rgba 0 0 0 0>}
       cornerRadii=<0 0 0 0>
      }
      tintColor {colorUseRule=UseColor_Specified, specifiedColor=<rgba 1 1 1 1>}
      uVRegion {bIsValid=false, max=<0 0>, min=<0 0>}
     }
     marqueeImage {
      bIsDynamicallyLoaded=false
      drawAs=Box
      imageType=FullColor
      mirroring=NoMirror
      resourceName=../../../Engine/Content/Slate/Common/ProgressBar_Marquee.png
      resourceObject=&MI_UI_Circle_ProgressBar
      tiling=Horizontal
      imageSize=<8 8>
      margin=<ltrb 0.5 0.5 0.5 0.5>
      outlineSettings {
       bUseBrushTransparency=false
       roundingType=HalfHeightRadius
       width=0
       color {colorUseRule=UseColor_Specified, specifiedColor=<rgba 0 0 0 0>}
       cornerRadii=<0 0 0 0>
      }
      tintColor {colorUseRule=UseColor_Specified, specifiedColor=<rgba 1 1 1 1>}
      uVRegion {bIsValid=false, max=<0 0>, min=<0 0>}
     }
    }
   }
   slot {class=CanvasPanelSlot, properties {layoutData {alignment=<0 0>, anchors {maximum=<1 1>, minimum=<0 0>}, offsets=<ltrb 0 2 0 2>}}}
  }
  AnimatedProgress {
   widgetClassPath=&ProgressBar
   bIsVariable=true
   parent=ProgressBarBox
   widget=AnimatedProgress
   properties {
    visibility=HitTestInvisible
    fillColorAndOpacity=<rgba 1 0.6038 0 1>
    widgetStyle {
     enableFillAnimation=true
     backgroundImage {
      bIsDynamicallyLoaded=false
      drawAs=Box
      imageType=FullColor
      mirroring=NoMirror
      resourceName=../../../Engine/Content/Slate/Common/ProgressBar_Background.png
      resourceObject=&MI_UI_Circle_ProgressBar
      tiling=NoTile
      imageSize=<8 8>
      margin=<ltrb 0.5 0.5 0.5 0.5>
      outlineSettings {
       bUseBrushTransparency=false
       roundingType=HalfHeightRadius
       width=0
       color {colorUseRule=UseColor_Specified, specifiedColor=<rgba 0 0 0 0>}
       cornerRadii=<0 0 0 0>
      }
      tintColor {colorUseRule=UseColor_Specified, specifiedColor=<rgba 0 0 0 0>}
      uVRegion {bIsValid=false, max=<0 0>, min=<0 0>}
     }
     fillImage {
      bIsDynamicallyLoaded=false
      drawAs=Box
      imageType=FullColor
      mirroring=NoMirror
      resourceName=../../../Engine/Content/Slate/Common/ProgressBar_Fill.png
      resourceObject=null
      tiling=NoTile
      imageSize=<8 8>
      margin=<ltrb 0.5 0.5 0.5 0.5>
      outlineSettings {
       bUseBrushTransparency=false
       roundingType=HalfHeightRadius
       width=0
       color {colorUseRule=UseColor_Specified, specifiedColor=<rgba 0 0 0 0>}
       cornerRadii=<0 0 0 0>
      }
      tintColor {colorUseRule=UseColor_Specified, specifiedColor=<rgba 1 1 1 1>}
      uVRegion {bIsValid=false, max=<0 0>, min=<0 0>}
     }
     marqueeImage {
      bIsDynamicallyLoaded=false
      drawAs=Box
      imageType=FullColor
      mirroring=NoMirror
      resourceName=../../../Engine/Content/Slate/Common/ProgressBar_Marquee.png
      resourceObject=&MI_UI_Circle_ProgressBar
      tiling=Horizontal
      imageSize=<8 8>
      margin=<ltrb 0.5 0.5 0.5 0.5>
      outlineSettings {
       bUseBrushTransparency=false
       roundingType=HalfHeightRadius
       width=0
       color {colorUseRule=UseColor_Specified, specifiedColor=<rgba 0 0 0 0>}
       cornerRadii=<0 0 0 0>
      }
      tintColor {colorUseRule=UseColor_Specified, specifiedColor=<rgba 1 1 1 1>}
      uVRegion {bIsValid=false, max=<0 0>, min=<0 0>}
     }
    }
   }
   slot {class=CanvasPanelSlot, properties {layoutData {alignment=<0 0>, anchors {maximum=<1 1>, minimum=<0 0>}, offsets=<ltrb 0 2 0 2>}}}
  }
  NumberDetailsSizeBox {widgetClassPath=&SizeBox, parent=AllDetailsBox, widget=NumberDetailsSizeBox, slot {class=VerticalBoxSlot}}
  NumberDetailsHorizontalBox {widgetClassPath=&HorizontalBox, parent=NumberDetailsSizeBox, widget=NumberDetailsHorizontalBox, slot {class=SizeBoxSlot}}
  LevelBox {
   widgetClassPath=&HorizontalBox
   parent=NumberDetailsHorizontalBox
   widget=LevelBox
   slot {class=HorizontalBoxSlot, properties {horizontalAlignment=HAlign_Left}}
  }
  Level {
   widgetClassPath=/Script/UMG.RichTextBlock
   bIsVariable=true
   parent=LevelBox
   widget=Level
   properties {
    text="lvl "
    textStyleSet=&/CreativeRichTextStat
    defaultTextStyleOverride {
     overflowPolicy=Clip
     transformPolicy=None
     colorAndOpacity {colorUseRule=UseColor_Specified, specifiedColor=<rgba 0.2195 0.6584 1 1>}
     font {
      bForceMonospaced=false
      bMaterialIsStencil=false
      fontMaterial=null
      fontObject=&BurbankSmall
      letterSpacing=0
      monospacedWidth=1
      size=24
      skewAmount=0
      typefaceFontName=Bold
      outlineSettings {
       bApplyOutlineToDropShadows=false
       bMiteredCorners=false
       bSeparateFillAlpha=false
       outlineMaterial=null
       outlineSize=0
       outlineColor=<rgba 0 0 0 1>
      }
     }
     highlightColor {colorUseRule=UseColor_Specified, specifiedColor=<rgba 0 0 0 1>}
     highlightShape {
      bIsDynamicallyLoaded=false
      drawAs=Image
      imageType=NoImage
      mirroring=NoMirror
      resourceName=None
      resourceObject=null
      tiling=NoTile
      imageSize=<32 32>
      margin=<ltrb 0 0 0 0>
      outlineSettings {
       bUseBrushTransparency=false
       roundingType=HalfHeightRadius
       width=0
       color {colorUseRule=UseColor_Specified, specifiedColor=<rgba 0 0 0 0>}
       cornerRadii=<0 0 0 0>
      }
      tintColor {colorUseRule=UseColor_Specified, specifiedColor=<rgba 1 1 1 1>}
      uVRegion {bIsValid=false, max=<0 0>, min=<0 0>}
     }
     selectedBackgroundColor {colorUseRule=UseColor_Foreground, specifiedColor=<rgba 1 0 1 1>}
     shadowColorAndOpacity=<rgba 0 0 0 1>
     shadowOffset=<0 0>
     strikeBrush {
      bIsDynamicallyLoaded=false
      drawAs=Image
      imageType=NoImage
      mirroring=NoMirror
      resourceName=None
      resourceObject=null
      tiling=NoTile
      imageSize=<32 32>
      margin=<ltrb 0 0 0 0>
      outlineSettings {
       bUseBrushTransparency=false
       roundingType=HalfHeightRadius
       width=0
       color {colorUseRule=UseColor_Specified, specifiedColor=<rgba 0 0 0 0>}
       cornerRadii=<0 0 0 0>
      }
      tintColor {colorUseRule=UseColor_Specified, specifiedColor=<rgba 1 1 1 1>}
      uVRegion {bIsValid=false, max=<0 0>, min=<0 0>}
     }
     underlineBrush {
      bIsDynamicallyLoaded=false
      drawAs=Image
      imageType=NoImage
      mirroring=NoMirror
      resourceName=None
      resourceObject=null
      tiling=NoTile
      imageSize=<32 32>
      margin=<ltrb 0 0 0 0>
      outlineSettings {
       bUseBrushTransparency=false
       roundingType=HalfHeightRadius
       width=0
       color {colorUseRule=UseColor_Specified, specifiedColor=<rgba 0 0 0 0>}
       cornerRadii=<0 0 0 0>
      }
      tintColor {colorUseRule=UseColor_Specified, specifiedColor=<rgba 1 1 1 1>}
      uVRegion {bIsValid=false, max=<0 0>, min=<0 0>}
     }
    }
   }
   slot {class=HorizontalBoxSlot}
  }
  Level_Number {
   widgetClassPath=&CommonTextBlock
   parent=LevelBox
   widget=Level_Number
   properties {
    text=9
    colorAndOpacity {colorUseRule=UseColor_Specified, specifiedColor=<rgba 0.6038 0.8879 1 1>}
    font {
     bForceMonospaced=false
     bMaterialIsStencil=false
     fontMaterial=null
     fontObject=&BurbankSmall
     letterSpacing=0
     monospacedWidth=1
     size=12
     skewAmount=0
     typefaceFontName=Bold
     outlineSettings {
      bApplyOutlineToDropShadows=false
      bMiteredCorners=false
      bSeparateFillAlpha=false
      outlineMaterial=null
      outlineSize=0
      outlineColor=<rgba 0 0 0 1>
     }
    }
    shadowOffset=<0 0>
   }
   slot {class=HorizontalBoxSlot}
  }
  LevelSeparator {
   widgetClassPath=&CommonTextBlock
   parent=LevelBox
   widget=LevelSeparator
   properties {
    text=/
    colorAndOpacity {colorUseRule=UseColor_Specified, specifiedColor=<rgba 0.2195 0.6584 1 1>}
    font {
     bForceMonospaced=false
     bMaterialIsStencil=false
     fontMaterial=null
     fontObject=&BurbankSmall
     letterSpacing=0
     monospacedWidth=1
     size=12
     skewAmount=0
     typefaceFontName=Bold
     outlineSettings {
      bApplyOutlineToDropShadows=false
      bMiteredCorners=false
      bSeparateFillAlpha=false
      outlineMaterial=null
      outlineSize=0
      outlineColor=<rgba 0 0 0 1>
     }
    }
    shadowOffset=<0 0>
   }
   slot {class=HorizontalBoxSlot}
  }
  MaxLevel_Number {
   widgetClassPath=&CommonTextBlock
   parent=LevelBox
   widget=MaxLevel_Number
   properties {
    text=100
    colorAndOpacity {colorUseRule=UseColor_Specified, specifiedColor=<rgba 0.2195 0.6584 1 1>}
    font {
     bForceMonospaced=false
     bMaterialIsStencil=false
     fontMaterial=null
     fontObject=&BurbankSmall
     letterSpacing=0
     monospacedWidth=1
     size=12
     skewAmount=0
     typefaceFontName=Bold
     outlineSettings {
      bApplyOutlineToDropShadows=false
      bMiteredCorners=false
      bSeparateFillAlpha=false
      outlineMaterial=null
      outlineSize=0
      outlineColor=<rgba 0 0 0 1>
     }
    }
    shadowOffset=<0 0>
   }
   slot {class=HorizontalBoxSlot}
  }
  ValueBox {
   widgetClassPath=&SizeBox
   parent=NumberDetailsHorizontalBox
   widget=ValueBox
   properties {bOverride_MaxDesiredWidth=true, maxDesiredWidth=200, widthOverride=200}
   slot {class=HorizontalBoxSlot, properties {horizontalAlignment=HAlign_Right, verticalAlignment=VAlign_Bottom, size {sizeRule=Fill, value=1}}}
  }
  ValueBoxSwitcher {widgetClassPath=&CommonVisibilitySwitcher, parent=ValueBox, widget=ValueBoxSwitcher, slot {class=SizeBoxSlot}}
  HorizontalBox_1 {widgetClassPath=&HorizontalBox, parent=ValueBoxSwitcher, widget=HorizontalBox_1, slot {class=CommonVisibilitySwitcherSlot}}
  Value_Number {
   widgetClassPath=&CommonTextBlock
   parent=HorizontalBox_1
   widget=Value_Number
   properties {
    text=12000
    colorAndOpacity {colorUseRule=UseColor_Specified, specifiedColor=<rgba 0.6038 0.8879 1 1>}
    font {
     bForceMonospaced=false
     bMaterialIsStencil=false
     fontMaterial=null
     fontObject=&BurbankSmall
     letterSpacing=0
     monospacedWidth=1
     size=12
     skewAmount=0
     typefaceFontName=Bold
     outlineSettings {
      bApplyOutlineToDropShadows=false
      bMiteredCorners=false
      bSeparateFillAlpha=false
      outlineMaterial=null
      outlineSize=0
      outlineColor=<rgba 0 0 0 1>
     }
    }
    shadowOffset=<0 0>
   }
   slot {class=HorizontalBoxSlot}
  }
  ValueSeparator {
   widgetClassPath=&CommonTextBlock
   bIsVariable=true
   parent=HorizontalBox_1
   widget=ValueSeparator
   properties {
    text=/
    colorAndOpacity {colorUseRule=UseColor_Specified, specifiedColor=<rgba 1 0.1314 0 1>}
    font {
     bForceMonospaced=false
     bMaterialIsStencil=false
     fontMaterial=null
     fontObject=&BurbankSmall
     letterSpacing=0
     monospacedWidth=1
     size=12
     skewAmount=0
     typefaceFontName=Bold
     outlineSettings {
      bApplyOutlineToDropShadows=false
      bMiteredCorners=false
      bSeparateFillAlpha=false
      outlineMaterial=null
      outlineSize=0
      outlineColor=<rgba 0 0 0 1>
     }
    }
    shadowOffset=<0 0>
   }
   slot {class=HorizontalBoxSlot}
  }
  ValueMax_Number {
   widgetClassPath=&CommonTextBlock
   bIsVariable=true
   parent=HorizontalBox_1
   widget=ValueMax_Number
   properties {
    text=12000
    colorAndOpacity {colorUseRule=UseColor_Specified, specifiedColor=<rgba 0.232 0.0643 -0.024 1>}
    font {
     bForceMonospaced=false
     bMaterialIsStencil=false
     fontMaterial=null
     fontObject=&BurbankSmall
     letterSpacing=0
     monospacedWidth=1
     size=12
     skewAmount=0
     typefaceFontName=Bold
     outlineSettings {
      bApplyOutlineToDropShadows=false
      bMiteredCorners=false
      bSeparateFillAlpha=false
      outlineMaterial=null
      outlineSize=0
      outlineColor=<rgba 0 0 0 1>
     }
    }
    shadowOffset=<0 0>
   }
   slot {class=HorizontalBoxSlot}
  }
  MaximumText {
   widgetClassPath=&CommonTextBlock
   bIsVariable=true
   parent=ValueBoxSwitcher
   widget=MaximumText
   properties {
    text=Max
    colorAndOpacity {colorUseRule=UseColor_Specified, specifiedColor=<rgba 0.232 0.0643 -0.024 1>}
    font {
     bForceMonospaced=false
     bMaterialIsStencil=false
     fontMaterial=null
     fontObject=&BurbankSmall
     letterSpacing=0
     monospacedWidth=1
     size=12
     skewAmount=0
     typefaceFontName=Bold
     outlineSettings {
      bApplyOutlineToDropShadows=false
      bMiteredCorners=false
      bSeparateFillAlpha=false
      outlineMaterial=null
      outlineSize=0
      outlineColor=<rgba 0 0 0 1>
     }
    }
    shadowOffset=<0 0>
   }
   slot {class=CommonVisibilitySwitcherSlot}
  }
  icon {
   widgetClassPath=&Image
   bIsVariable=true
   parent=AllContainer
   widget=icon
   properties {brush {
     bIsDynamicallyLoaded=false
     drawAs=Image
     imageType=NoImage
     mirroring=NoMirror
     resourceName=None
     resourceObject=&/M_PlayerStatIcon
     tiling=NoTile
     imageSize=<44 44>
     margin=<ltrb 0 0 0 0>
     outlineSettings {
      bUseBrushTransparency=false
      roundingType=HalfHeightRadius
      width=0
      color {colorUseRule=UseColor_Specified, specifiedColor=<rgba 0 0 0 0>}
      cornerRadii=<0 0 0 0>
     }
     tintColor {colorUseRule=UseColor_Specified, specifiedColor=<rgba 1 1 1 1>}
     uVRegion {bIsValid=false, max=<0 0>, min=<0 0>}
    }}
   slot {class=OverlaySlot, properties {verticalAlignment=VAlign_Center, padding=<ltrb 10 0 0 0>}}
  }
 ]
}
